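# Creative-mode HUD quickbar, derived from QuickbarBase, with a phone slot and eight further slots.
# Creative HUD context, input and inventory events drive slot-row visibility, phone padding around the health bar, input-specific hold prompts and sounds.
# The slot row shows only while the quickbar is equipped.
# A looping pulse highlights the phone until the tutorial is seen, with equipping or a failsafe timer hiding it.
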WidgetBlueprint /FortniteCoreUI/HUD/CreativeQuickbarPrimary
animations [Idle {
  binding_count=1
  length_seconds=0.5
  object=&self:Idle
  sequence {bindings [IMG_Outline {
     id=#1
     display_name=Overlay_CTA
     possessed_class=Image
     tracks [
      {
       type=MovieScene2DTransformTrack
       sections [{
         type=MovieScene2DTransformSection
         end_frame=10
         start_frame=0
         channels [
          Translation.X {type=MovieSceneScriptingFloatChannel}
          Translation.Y {type=MovieSceneScriptingFloatChannel}
          Angle {type=MovieSceneScriptingFloatChannel}
          Scale.X {type=MovieSceneScriptingFloatChannel, keys [[0,1],[10,1.5]]}
          Scale.Y {type=MovieSceneScriptingFloatChannel, keys [[0,1],[10,1.5]]}
          Shear.X {type=MovieSceneScriptingFloatChannel}
          Shear.Y {type=MovieSceneScriptingFloatChannel}
         ]
        }]
      }
      {
       type=MovieSceneFloatTrack
       display_name=Render Opacity
       sections [{type=MovieSceneFloatSection, end_frame=10, start_frame=0, channels [None {type=MovieSceneScriptingFloatChannel, keys [[0,1],[10,0]]}]}]
      }
     ]
    }]}
 }]
graphs [
 EnablePhoneMode {nodes [
   n1 = EnablePhoneMode!() -> (then, PhoneModeEnabled:bool=~) {is=title K2Node_FunctionEntry}
   > Set IsPhoneEnabled(IsPhoneEnabled:bool=@n1) -> (Output_Get:bool)
   > UpdatePhoneEnabled()
  ]}
 EventGraph {nodes [
   Event OnCreativeModeEnabledChanged!() -> (OutputDelegate, then, IsEnabled:bool=~) {is=title K2Node_CustomEvent}
   > UpdateKeyBindingText(self:Fort Quick Bar Slot Base Object Reference=Get QuickbarSlot1() -> (QuickbarSlot1:Athena Quick Bar Slot Object Reference))
   > UpdatePhoneEnabled()
   n1 = OnCreativeSwapPhoneHoldChanged!() -> (OutputDelegate, then, bIsSwapped:bool=~) {is=title K2Node_CustomEvent}
   > UpdateKeyBindingText(self=Get QuickbarSlot1())
   n2 = OnQuickBarSlotFocusChanged_Bind!() -> (OutputDelegate, then, QuickbarIndex:EFortQuickBars Enum=Primary, Slot:int) {is=title K2Node_CustomEvent}
   > Branch(Condition=AND Boolean(
     A="Equal (Enum)"(A:EFortQuickBars Enum=@n2.QuickbarIndex, B:EFortQuickBars Enum=Creative)
     B="Not Equal (Integer)"!(A:int=@n2.Slot, B:int=0) -> (ReturnValue:bool) {is=title K2Node_CallFunction}
    ))
   > PlaySound2D(Sound=/Game/Sounds/Creative/Gadgets/MoveTool/MoveTool_Highlight_03_Cue)
   n3 = HandleInputChanged!() -> (OutputDelegate, then, bNewInputType:ECommonInputType Enum) {is=title K2Node_CustomEvent}
   > UpdateTextMessage!(execute, self, inInputType:ECommonInputType Enum=@n3.bNewInputType) -> (then)
   n4 = Event OnQuickbarContentsChanged!() -> (OutputDelegate, then, QuickbarIndex:EFortQuickBars Enum=Primary, ChangedSlots:Array of Integers) {is=title K2Node_Event}
   n5 = Get FocusedSlot() -> (FocusedSlot:int=0)
   n6 = Get QuickbarSlots() -> (QuickbarSlots:Array of Fort Quick Bar Slot Base Object References)
   > Branch(Condition=AND Boolean(
     A="Equal (Enum)"(A=@n4.QuickbarIndex, B=Creative)
     B=Is Valid Index!(TargetArray:Array of Fort Quick Bar Slot Base Object References=@n6, IndexToTest:int=@n5) -> (ReturnValue:bool)
    ))
   > PlaySound2D(Sound=/Game/Sounds/Creative/Gadgets/MoveTool/MoveTool_QuickBar_Add_Cue)
   Event HandleKeybindsChanged() {is=title K2Node_Event}
   > Parent: HandleKeybindsChanged() {is=title K2Node_CallParentFunction}
   > UpdatePhoneEnabled()
   > UpdatePhoneClutchTextVisibility()
   Event Destruct()
   > Parent: Destruct()
   n7 = Create Event()
   n8 = Get CreativeHUDLContext!() -> (ReturnValue:Creative HUDLContext Object Reference) {is=title K2Node_GetBlueprintContextEx}
   > Unbind Event from On Creative Mode Enabled Changed(self:Creative HUDLContext Object Reference=@n8, Delegate=@n7) {is=title K2Node_RemoveDelegate}
   n9 = Create Event()
   > Unbind Event from On Creative Quickbar Enabled Changed(self:Creative HUDLContext Object Reference=@n8, Delegate=@n9) {is=title K2Node_RemoveDelegate}
   n10 = Create Event()
   > Unbind Event from On Creative Quickbar Active Changed(self:Creative HUDLContext Object Reference=@n8, Delegate=@n10) {is=title K2Node_RemoveDelegate}
   Event Construct()
   > n11 = Sequence!(execute) -> (then_0, then_1, then_2:Exec, then_3:Exec, then_4:Exec, then_5:Exec, then_6:Exec)
   Parent: Construct(execute=@n11.then_1)
   n12 = Cast To FortPlayerControllerGameplay!(execute=@n11.then_4, Object=GetOwningPlayer()) -> (then, CastFailed, AsFort Player Controller Gameplay:Fort Player Controller Gameplay Object Reference) {is=title K2Node_DynamicCast}
   > Set IsQuickbarEnabled(IsQuickbarEnabled:bool=IsCreativeQuickbarEnabled!(self:Fort Player Controller Gameplay Object Reference=@n12) -> (ReturnValue:bool)) -> (Output_Get:bool)
   > n13 = Set IsQuickbarActive(IsQuickbarActive:bool=IsCreativeQuickbarActive!(self:Fort Player Controller Gameplay Object Reference=@n12) -> (ReturnValue:bool)) -> (Output_Get:bool)
   > OnCreativeQuickbarActiveChanged!(execute, self, IsActive:bool=@n13) -> (then)
   > UpdatePhoneEnabled()
   > Bind Event to On Creative Swap Phone and Harvest Hold Changed Delegate(self:Fort Player Controller Gameplay Object Reference=@n12, Delegate=@n1) {is=title K2Node_AddDelegate}
   SetBoundAction!(
    execute=@n11.then_5
    self:Fort Keybind Widget Object Reference=Get ClutchKeybindWidget() -> (ClutchKeybindWidget:Keybind Widget Object Reference)
    NewBoundAction:Name=GetCreativeQuickbarActivateClutchName!(self:Fort Global UIContext Object Reference=Get FortGlobalUIContext!() -> (ReturnValue:Fort Global UIContext Object Reference) {is=title K2Node_GetBlueprintContextEx}) -> (ReturnValue:Name=None)
   ) -> (then)
   n14 = Get CreativeHUDLContext()
   Bind Event to On Creative Mode Enabled Changed(execute=@n11.then_6, self:Creative HUDLContext Object Reference=@n14, Delegate=@n7) {is=title K2Node_AddDelegate}
   > Bind Event to On Creative Quickbar Enabled Changed(self:Creative HUDLContext Object Reference=@n14, Delegate=@n9) {is=title K2Node_AddDelegate}
   > Bind Event to On Creative Quickbar Active Changed(self:Creative HUDLContext Object Reference=@n14, Delegate=@n10) {is=title K2Node_AddDelegate}
   > Bind Event to On Creative Quickbar Equipped Changed(self:Creative HUDLContext Object Reference=@n14, Delegate=Create Event()) {is=title K2Node_AddDelegate}
   > Assign On Input Method Changed!(execute:Exec, self:Common Input Object Reference=Get CommonInputSubsystem(), Delegate:"Delegate (by ref)"=@n3) -> (then:Exec) {is=title K2Node_AssignDelegate}
   > InitTutorialCTA()
   Set QuickbarSlots(
    execute=@n11.then_0
    QuickbarSlots:Array of Fort Quick Bar Slot Base Object References=Make Array!(
     "[0]":Fort Quick Bar Slot Base Object Reference=Get QuickbarSlot1()
     "[1]":Fort Quick Bar Slot Base Object Reference=Get QuickbarSlot2() -> (QuickbarSlot2:Creative Quickbar Slot Object Reference)
     "[2]":Fort Quick Bar Slot Base Object Reference=Get QuickbarSlot3() -> (QuickbarSlot3:Creative Quickbar Slot Object Reference)
     "[3]":Fort Quick Bar Slot Base Object Reference=Get QuickbarSlot4() -> (QuickbarSlot4:Creative Quickbar Slot Object Reference)
     "[4]":Fort Quick Bar Slot Base Object Reference=Get QuickbarSlot5() -> (QuickbarSlot5:Creative Quickbar Slot Object Reference)
     "[5]":Fort Quick Bar Slot Base Object Reference=Get QuickbarSlot6() -> (QuickbarSlot6:Creative Quickbar Slot Object Reference)
     "[6]":Fort Quick Bar Slot Base Object Reference=Get QuickbarSlot7() -> (QuickbarSlot7:Creative Quickbar Slot Object Reference)
     "[7]":Fort Quick Bar Slot Base Object Reference=Get QuickbarSlot8() -> (QuickbarSlot8:Creative Quickbar Slot Object Reference)
     "[8]":Fort Quick Bar Slot Base Object Reference=Get QuickbarSlot9() -> (QuickbarSlot9:Creative Quickbar Slot Object Reference)
    ) -> (Array:Array of Fort Quick Bar Slot Base Object References)
   ) -> (Output_Get:Array of Fort Quick Bar Slot Base Object References)
   n15 = GetQuickbarFocus!(self:Fort Inventory Context Object Reference=Get FortInventoryContext!() -> (ReturnValue:Fort Inventory Context Object Reference) {is=title K2Node_GetBlueprintContextEx}) -> (OutQuickBar:EFortQuickBars Enum=Primary, OutSlot:int=0, OutSecondarySlot:int=0, OutPreviousFocusedSlot:int=0)
   OnQuickbarSlotFocusChanged!(execute=@n11.then_2, self:Fort Quickbar Base Object Reference, QuickbarIndex:EFortQuickBars Enum=@n15, SlotIndex:int=@n15.OutSlot) -> (then)
   Assign On Quickbar Slot Focus Changed!(
    execute:Exec=@n11.then_3
    self:Fort Inventory Context Object Reference=Get FortInventoryContext()
    Delegate:"Delegate (by ref)"=Create Event()
   ) -> (then:Exec) {is=title K2Node_AssignDelegate}
  ]}
 HandleCtaFailsafeTimer {nodes [
   HandleCtaFailsafeTimer() {is=title K2Node_FunctionEntry}
   > SetVisibility(
    self=Get Slot_BehindImage!(self:Athena Quickbar Slot Base Object Reference=Get QuickbarSlot1()) -> (Slot_BehindImage:Named Slot Object Reference)
    InVisibility=Collapsed
   )
   > Clear and Invalidate Timer by Handle(Handle=Get CtaFailsafeTimer() -> (CtaFailsafeTimer:Timer Handle Structure))
  ]}
 InitTutorialCTA {nodes [
   InitTutorialCTA() {is=title K2Node_FunctionEntry}
   > n1 = Branch(Condition=GetHasSeenCreativePhoneTutorial!(self:Fort Game User Settings Object Reference=GetFortGameUserSettings!() -> (ReturnValue:Fort Game User Settings Object Reference)) -> (ReturnValue:bool))
   n2 = Get Slot_BehindImage(self=Get QuickbarSlot1())
   > SetVisibility(self=@n2, InVisibility=Collapsed)
   SetVisibility(execute=@n1.else, self=@n2, InVisibility=HitTestInvisible)
   > PlayAnimation(InAnimation=Get Idle() -> (Idle:Widget Animation Object Reference), NumLoopsToPlay=0)
   > Clear and Invalidate Timer by Handle(Handle=Get CtaFailsafeTimer())
   > n3 = Set Timer by Event(Delegate=Create Event(), Time=120, bLooping=false)
   > Set CtaFailsafeTimer(CtaFailsafeTimer:Timer Handle Structure=@n3) -> (Output_Get:Timer Handle Structure)
  ]}
 IsHealthbarHidden {nodes [
   IsHealthbarHidden() {is=title K2Node_FunctionEntry}
   n1 = GetOwningPlayer()
   > n2 = Is Valid(InputObject=@n1)
   n3 = Get Controlled Pawn!(self:Controller Object Reference=@n1) -> (ReturnValue:Pawn Object Reference) {is=title K2Node_CallFunction}
   > n4 = Is Valid(InputObject=@n3)
   n5 = GetGameStateAthena!() -> (ReturnValue:Fort Game State Athena Object Reference)
   > n6 = Is Valid(InputObject=@n5)
   > Return Node!(
    execute
    HealthbarHidden:bool=HasAnyTags!(
     TagContainer:"Gameplay Tag Container Structure (by ref)"=GetHUDElementsToHide!(self:Fort Game State Object Reference=@n5, Pawn:Pawn Object Reference=@n3) -> (ReturnValue:Gameplay Tag Container Structure)
     OtherContainer:"Gameplay Tag Container Structure (by ref)"=MakeLiteralGameplayTagContainer!(Value:Gameplay Tag Container Structure=(GameplayTags=((TagName="HUD.TopLeft.PlayerHitPoints")))) -> (ReturnValue:Gameplay Tag Container Structure)
     bExactMatch:bool=true
    ) -> (ReturnValue:bool)
   )
   Return Node(execute=[@n2.Is Not Valid, @n4.Is Not Valid, @n6.Is Not Valid], HealthbarHidden=false)
  ]}
 OnCreativeQuickbarActiveChanged {nodes [
   n1 = OnCreativeQuickbarActiveChanged!() -> (then, IsActive:bool=~) {is=title K2Node_FunctionEntry}
   > n2 = Set IsQuickbarActive(IsQuickbarActive=@n1)
   > SetVisibility(
    self=Get HoldToReplaceTextbox() -> (HoldToReplaceTextbox:Text Object Reference)
    InVisibility=Select(Option 0=Collapsed, Option 1=Visible, Index=@n2)
   )
   > UpdatePhoneClutchTextVisibility()
  ]}
 OnCreativeQuickbarEnabledChanged {nodes [
   n1 = OnCreativeQuickbarEnabledChanged!() -> (then, bIsEnabled:bool=~) {is=title K2Node_FunctionEntry}
   > Set IsQuickbarEnabled(IsQuickbarEnabled=@n1)
   > UpdatePhoneEnabled()
  ]}
 OnCreativeQuickbarEquippedChanged {nodes [
   n1 = OnCreativeQuickbarEquippedChanged!() -> (then, bIsEquipped:bool=~) {is=title K2Node_FunctionEntry}
   > n2 = Set IsQuickbarEquipped(IsQuickbarEquipped:bool=@n1) -> (Output_Get:bool)
   > UpdatePhoneEnabled()
   > n3 = Branch(Condition=@n2)
   > SetVisibility(self=Get Slot_BehindImage(self=Get QuickbarSlot1()), InVisibility=Collapsed)
   InitTutorialCTA(execute=@n3.else) {is=title K2Node_CallFunction}
  ]}
 SetPhonePosition {nodes [
   SetPhonePosition() {is=title K2Node_FunctionEntry}
   > n1 = IsNewHUDEnabled!(execute, WorldContextObject:Object Reference=Self-Reference()) -> (then, IsNewHUD:bool)
   > n2 = IsHealthbarHidden!(execute, self) -> (then, HealthbarHidden:bool) {is=title K2Node_CallFunction}
   n3 = NOT Boolean(A=@n2)
   n4 = NOT Boolean(A=Get IsQuickbarEquipped() -> (IsQuickbarEquipped:bool))
   > SetPadding!(
    execute
    self:User Widget Object Reference=Get QuickbarSlot1()
    InPadding_Left:float=0
    InPadding_Top:float=0
    InPadding_Right:float=0
    InPadding_Bottom:float=float * float(A=72, B:bool=AND Boolean!(A=@n3, B=@n1, C:bool=@n4) -> (ReturnValue)) {is=title K2Node_PromotableOperator}
   ) -> (then)
  ]}
 UpdatePhoneClutchTextVisibility {nodes [
   UpdatePhoneClutchTextVisibility() {is=title K2Node_FunctionEntry}
   n1 = NOT Boolean(A=Get IsPhoneEnabled() -> (IsPhoneEnabled:bool))
   n2 = NOT Boolean(A=IsUsingGamepad!(self:Fort Global UIContext Object Reference=Get FortGlobalUIContext() -> (ReturnValue:Fort Global UIContext Object Reference) {is=title K2Node_GetSubsystem}) -> (ReturnValue:bool))
   n3 = Get IsQuickbarActive() -> (IsQuickbarActive:bool)
   > SetVisibility(
    self=Get Overlay_GamepadClutchPrompt() -> (Overlay_GamepadClutchPrompt:Overlay Object Reference)
    InVisibility=Select(Option 0=Visible, Option 1=Hidden, Index=OR Boolean!(A=@n1, B=@n3, C:bool=@n2) -> (ReturnValue))
   )
  ]}
 UpdatePhoneEnabled {nodes [
   UpdatePhoneEnabled() {is=title K2Node_FunctionEntry}
   > n1 = Branch(Condition=Get IsQuickbarEquipped())
   n2 = Get VBox_Slots() -> (VBox_Slots:Vertical Box Object Reference)
   > n3 = SetVisibility(self=@n2, InVisibility=SelfHitTestInvisible)
   SetPhonePosition(execute=[@n3.then, @n4.then]) {is=title K2Node_CallFunction}
   n4 = SetVisibility(execute=@n1.else, self=@n2, InVisibility=Collapsed)
  ]}
 UpdateTextMessage {nodes [
   n1 = UpdateTextMessage!() -> (then, inInputType:ECommonInputType Enum) {is=title K2Node_FunctionEntry}
   > n2 = Switch on ECommonInputType!(execute, Selection:ECommonInputType Enum=@n1) -> (MouseAndKeyboard:Exec, Gamepad:Exec, Touch:Exec, Count:Exec) {is=title K2Node_SwitchEnum}
   n3 = Get HoldToReplaceTextbox()
   "SetText (Text)"(execute=@n2.Gamepad, self=@n3, InText="LOCTEXT(\"Hold controller button to replace slot\")")
   "SetText (Text)"(execute=@n2.MouseAndKeyboard, self=@n3, InText="LOCTEXT(\"Hold key to replace slot\")")
   "SetText (Text)"(execute=@n2.Touch, self=@n3, InText="LOCTEXT(\"Tocuh and hold to replace slot\")")
  ]}
]
variables [
 CtaFailsafeTimer {type=TimerHandle}
 IsPhoneEnabled {type=boolean}
 IsQuickbarActive {type=boolean}
 IsQuickbarEnabled {type=boolean}
 IsQuickbarEquipped {type=boolean}
]
widgets {
 info {
  inheritedWidgetCount=2
  namedSlotCount=1
  parentClass=/Game/UI/InGame/HUD/QuickbarBase.QuickbarBase_C
  rootWidgetClass=&Overlay
  widgetCount=33
 }
 widgets [
  Overlay_2 {widgetClassPath=&Overlay, widget=Overlay_2}
  Cr_TutorialSimpleMessage {
   widgetClassPath=/Game/Athena/UI/InGame/Creative/QuickbarTutorial/Cr_TutorialSimpleMessage
   parent=Overlay_2
   widget=Cr_TutorialSimpleMessage
   properties {visibility=Collapsed, renderTransform {angle=0, scale=<1 1>, shear=<0 0>, translation=<160 -240>}}
   slot {class=OverlaySlot}
  }
  HorizontalBox_140 {widgetClassPath=&HorizontalBox, parent=Overlay_2, widget=HorizontalBox_140, slot {class=OverlaySlot}}
  HorizontalBox_Quickbars {
   widgetClassPath=&HorizontalBox
   bIsVariable=true
   parent=HorizontalBox_140
   widget=HorizontalBox_Quickbars
   slot {class=HorizontalBoxSlot, properties {horizontalAlignment=HAlign_Left, verticalAlignment=VAlign_Top}}
  }
  Overlay_0 {
   widgetClassPath=&Overlay
   parent=HorizontalBox_Quickbars
   widget=Overlay_0
   slot {class=HorizontalBoxSlot, properties {verticalAlignment=VAlign_Bottom, padding=<ltrb 0 0 8 0>}}
  }
  QuickbarSlot1 {
   widgetClassPath=/Game/Athena/HUD/AthenaQuickBarSlot
   bIsVariable=true
   parent=Overlay_0
   widget=QuickbarSlot1
   properties {quickBarType=Creative, visibility=Collapsed, cooldownTypesSupported [Death]}
   slot {class=OverlaySlot}
  }
  Overlay_CTA {widgetClassPath=&Overlay, namedSlotHost=QuickbarSlot1, widget=Overlay_CTA, properties {visibility=HitTestInvisible}}
  IMG_Center {
   widgetClassPath=&Image
   parent=Overlay_CTA
   widget=IMG_Center
   properties {
    visibility=HitTestInvisible
    brush {
     bIsDynamicallyLoaded=false
     drawAs=Image
     imageType=NoImage
     mirroring=NoMirror
     resourceName=None
     resourceObject=/Game/Athena/HUD/Art/Materials/MI_HUD_Thumbstick_Blurry
     tiling=NoTile
     imageSize=<64 64>
     margin=<ltrb 0 0 0 0>
     outlineSettings {
      bUseBrushTransparency=false
      roundingType=HalfHeightRadius
      width=0
      color {colorUseRule=UseColor_Specified, specifiedColor=<rgba 0 0 0 0>}
      cornerRadii=<0 0 0 0>
     }
     tintColor {colorUseRule=UseColor_Specified, specifiedColor=<rgba 0.9131 0.6308 0.0865 1>}
     uVRegion {bIsValid=false, max=<0 0>, min=<0 0>}
    }
    renderTransform {angle=0, scale=<2 2>, shear=<0 0>, translation=<0 0>}
   }
   slot {class=OverlaySlot, properties {horizontalAlignment=HAlign_Fill, verticalAlignment=VAlign_Fill}}
  }
  IMG_Outline {
   widgetClassPath=&Image
   parent=Overlay_CTA
   widget=IMG_Outline
   properties {
    visibility=HitTestInvisible
    brush {
     bIsDynamicallyLoaded=false
     drawAs=Border
     imageType=NoImage
     mirroring=NoMirror
     resourceName=None
     resourceObject=/Game/Athena/HUD/Art/Tutorial/FTUE_Outline_Square
     tiling=NoTile
     imageSize=<64 64>
     margin=<ltrb 0.1 0.1 0.1 0.1>
     outlineSettings {
      bUseBrushTransparency=false
      roundingType=HalfHeightRadius
      width=0
      color {colorUseRule=UseColor_Specified, specifiedColor=<rgba 0 0 0 0>}
      cornerRadii=<0 0 0 0>
     }
     tintColor {colorUseRule=UseColor_Specified, specifiedColor=<rgba 0.9137 0.6314 0.0863 1>}
     uVRegion {bIsValid=false, max=<0 0>, min=<0 0>}
    }
   }
   slot {class=OverlaySlot, properties {horizontalAlignment=HAlign_Fill, verticalAlignment=VAlign_Fill}}
  }
  VBox_Slots {
   widgetClassPath=&VerticalBox
   bIsVariable=true
   parent=HorizontalBox_Quickbars
   widget=VBox_Slots
   slot {class=HorizontalBoxSlot, properties {padding=<ltrb 20 0 0 0>}}
  }
  Overlay_1 {widgetClassPath=&Overlay, parent=VBox_Slots, widget=Overlay_1, slot {class=VerticalBoxSlot, properties {padding=<ltrb 0 0 0 20>}}}
  Overlay_Slots {widgetClassPath=&Overlay, parent=Overlay_1, widget=Overlay_Slots, slot {class=OverlaySlot}}
  ViolatorTail {
   widgetClassPath=&Image
   parent=Overlay_Slots
   widget=ViolatorTail
   properties {
    brush {
     bIsDynamicallyLoaded=false
     drawAs=Image
     imageType=NoImage
     mirroring=NoMirror
     resourceName=None
     resourceObject=/FortniteCoreUI/HUD/LTM/Playground/Materials/M_UI_Creative_LiveViolatorTail_Quickbar
     tiling=NoTile
     imageSize=<32 32>
     margin=<ltrb 0 0 0 0>
     outlineSettings {
      bUseBrushTransparency=false
      roundingType=HalfHeightRadius
      width=0
      color {colorUseRule=UseColor_Specified, specifiedColor=<rgba 0 0 0 0>}
      cornerRadii=<0 0 0 0>
     }
     tintColor {colorUseRule=UseColor_Specified, specifiedColor=<rgba 0 0.0529 0.2346 1>}
     uVRegion {bIsValid=false, max=<0 0>, min=<0 0>}
    }
    renderTransform {angle=114, scale=<-2 2>, shear=<0 0>, translation=<-22 35>}
   }
   slot {class=OverlaySlot, properties {verticalAlignment=VAlign_Center}}
  }
  Backplate {
   widgetClassPath=&Image
   parent=Overlay_Slots
   widget=Backplate
   properties {
    brush {
     bIsDynamicallyLoaded=false
     drawAs=Image
     imageType=NoImage
     mirroring=NoMirror
     resourceName=None
     resourceObject=/Game/Athena/HUD/LTM/Playground/Materials/M_UI_Creative_SkewBox_QuickBarBackplate
     tiling=NoTile
     imageSize=<32 32>
     margin=<ltrb 0 0 0 0>
     outlineSettings {
      bUseBrushTransparency=false
      roundingType=HalfHeightRadius
      width=0
      color {colorUseRule=UseColor_Specified, specifiedColor=<rgba 0 0 0 0>}
      cornerRadii=<0 0 0 0>
     }
     tintColor {colorUseRule=UseColor_Specified, specifiedColor=<rgba 1 1 1 1>}
     uVRegion {bIsValid=false, max=<0 0>, min=<0 0>}
    }
    renderTransform {angle=0, scale=<1 0.9>, shear=<0 0>, translation=<0 25>}
   }
   slot {class=OverlaySlot, properties {horizontalAlignment=HAlign_Fill, verticalAlignment=VAlign_Fill}}
  }
  HorizontalBox_147 {
   widgetClassPath=&HorizontalBox
   parent=Overlay_Slots
   widget=HorizontalBox_147
   slot {class=OverlaySlot, properties {padding=<ltrb 15 0 15 0>}}
  }
  QuickbarSlot2 {
   widgetClassPath=/Game/Athena/HUD/CreativeQuickbarSlot
   bIsVariable=true
   parent=HorizontalBox_147
   widget=QuickbarSlot2
   properties {slotIndex=1, visibility=Hidden}
   slot {class=HorizontalBoxSlot, properties {horizontalAlignment=HAlign_Left, verticalAlignment=VAlign_Top}}
  }
  QuickbarSlot3 {
   widgetClassPath=&CreativeQuickbarSlot
   bIsVariable=true
   parent=HorizontalBox_147
   widget=QuickbarSlot3
   properties {slotIndex=2, visibility=Hidden}
   slot {class=HorizontalBoxSlot, properties {horizontalAlignment=HAlign_Left, verticalAlignment=VAlign_Top}}
  }
  QuickbarSlot4 {
   widgetClassPath=&CreativeQuickbarSlot
   bIsVariable=true
   parent=HorizontalBox_147
   widget=QuickbarSlot4
   properties {slotIndex=3, visibility=Hidden}
   slot {class=HorizontalBoxSlot, properties {horizontalAlignment=HAlign_Left, verticalAlignment=VAlign_Top}}
  }
  QuickbarSlot5 {
   widgetClassPath=&CreativeQuickbarSlot
   bIsVariable=true
   parent=HorizontalBox_147
   widget=QuickbarSlot5
   properties {slotIndex=4, visibility=Hidden}
   slot {class=HorizontalBoxSlot, properties {horizontalAlignment=HAlign_Left, verticalAlignment=VAlign_Top}}
  }
  QuickbarSlot6 {
   widgetClassPath=&CreativeQuickbarSlot
   bIsVariable=true
   parent=HorizontalBox_147
   widget=QuickbarSlot6
   properties {slotIndex=5, visibility=Hidden}
   slot {class=HorizontalBoxSlot, properties {horizontalAlignment=HAlign_Left, verticalAlignment=VAlign_Top}}
  }
  QuickbarSlot7 {
   widgetClassPath=&CreativeQuickbarSlot
   bIsVariable=true
   parent=HorizontalBox_147
   widget=QuickbarSlot7
   properties {slotIndex=6, visibility=Hidden}
   slot {class=HorizontalBoxSlot, properties {horizontalAlignment=HAlign_Left, verticalAlignment=VAlign_Top}}
  }
  QuickbarSlot8 {
   widgetClassPath=&CreativeQuickbarSlot
   bIsVariable=true
   parent=HorizontalBox_147
   widget=QuickbarSlot8
   properties {slotIndex=7, visibility=Hidden}
   slot {class=HorizontalBoxSlot, properties {horizontalAlignment=HAlign_Left, verticalAlignment=VAlign_Top}}
  }
  QuickbarSlot9 {
   widgetClassPath=&CreativeQuickbarSlot
   bIsVariable=true
   parent=HorizontalBox_147
   widget=QuickbarSlot9
   properties {slotIndex=8, visibility=Hidden}
   slot {class=HorizontalBoxSlot, properties {horizontalAlignment=HAlign_Left, verticalAlignment=VAlign_Top}}
  }
  Overlay_Prompt {
   widgetClassPath=&Overlay
   parent=Overlay_1
   widget=Overlay_Prompt
   properties {renderTransform {angle=0, scale=<1 1>, shear=<0 0>, translation=<0 55>}}
   slot {class=OverlaySlot, properties {horizontalAlignment=HAlign_Center, verticalAlignment=VAlign_Bottom}}
  }
  HoldToReplaceTextbox {
   widgetClassPath=&TextBlock
   bIsVariable=true
   parent=Overlay_Prompt
   widget=HoldToReplaceTextbox
   properties {
    text=Hold to replace
    font {
     bForceMonospaced=false
     bMaterialIsStencil=false
     fontMaterial=null
     fontObject=&BurbankSmall
     letterSpacing=0
     monospacedWidth=1
     size=18
     skewAmount=0
     typefaceFontName=Bold
     outlineSettings {
      bApplyOutlineToDropShadows=false
      bMiteredCorners=false
      bSeparateFillAlpha=false
      outlineMaterial=null
      outlineSize=1
      outlineColor=<rgba 0 0 0 0.5>
     }
    }
    shadowColorAndOpacity=<rgba 0 0 0 0.5>
    shadowOffset=<2 1>
   }
   slot {class=OverlaySlot, properties {horizontalAlignment=HAlign_Center, verticalAlignment=VAlign_Center, padding=<ltrb 15 10 15 10>}}
  }
  Overlay_GamepadClutchPrompt {
   widgetClassPath=&Overlay
   bIsVariable=true
   parent=Overlay_1
   widget=Overlay_GamepadClutchPrompt
   properties {visibility=Collapsed, renderTransform {angle=0, scale=<1 1>, shear=<0 0>, translation=<0 -10>}}
   slot {class=OverlaySlot, properties {padding=<ltrb 10 0 0 0>}}
  }
  ButtonBackplate {
   widgetClassPath=&Image
   parent=Overlay_GamepadClutchPrompt
   widget=ButtonBackplate
   properties {brush {
     bIsDynamicallyLoaded=false
     drawAs=Image
     imageType=NoImage
     mirroring=NoMirror
     resourceName=None
     resourceObject=/FortniteCoreUI/Athena/HUD/LTM/Playground/Materials/M_UI_Creative_SkewBox_QuickBarPrompt
     tiling=NoTile
     imageSize=<32 40>
     margin=<ltrb 0 0 0 0>
     outlineSettings {
      bUseBrushTransparency=false
      roundingType=HalfHeightRadius
      width=0
      color {colorUseRule=UseColor_Specified, specifiedColor=<rgba 0 0 0 0>}
      cornerRadii=<0 0 0 0>
     }
     tintColor {colorUseRule=UseColor_Specified, specifiedColor=<rgba 1 1 1 1>}
     uVRegion {bIsValid=false, max=<0 0>, min=<0 0>}
    }}
   slot {class=OverlaySlot, properties {horizontalAlignment=HAlign_Fill, verticalAlignment=VAlign_Center}}
  }
  ClutchKeybinding {
   widgetClassPath=&HorizontalBox
   parent=Overlay_GamepadClutchPrompt
   widget=ClutchKeybinding
   slot {class=OverlaySlot, properties {horizontalAlignment=HAlign_Center, verticalAlignment=VAlign_Center, padding=<ltrb 12 4 12 4>}}
  }
  ClutchKeybindWidget {
   widgetClassPath=/FortUILibrary/Input/KeybindWidget
   bIsVariable=true
   parent=ClutchKeybinding
   widget=ClutchKeybindWidget
   slot {class=HorizontalBoxSlot, properties {horizontalAlignment=HAlign_Left}}
  }
  TextBlock_135 {
   widgetClassPath=&TextBlock
   parent=ClutchKeybinding
   widget=TextBlock_135
   properties {
    text=Hold
    font {
     bForceMonospaced=false
     bMaterialIsStencil=false
     fontMaterial=null
     fontObject=&Roboto
     letterSpacing=0
     monospacedWidth=1
     size=16
     skewAmount=0
     typefaceFontName=Bold
     outlineSettings {
      bApplyOutlineToDropShadows=false
      bMiteredCorners=false
      bSeparateFillAlpha=false
      outlineMaterial=null
      outlineSize=0
      outlineColor=<rgba 0 0 0 1>
     }
    }
   }
   slot {class=HorizontalBoxSlot, properties {horizontalAlignment=HAlign_Left, verticalAlignment=VAlign_Center}}
  }
  Spacer_45 {widgetClassPath=&Spacer, parent=VBox_Slots, widget=Spacer_45, properties {size=<1 40>}, slot {class=VerticalBoxSlot}}
  KeybindWidget_SwitchModeKeybind {widgetClassPath=/Script/FortniteUI.FortKeybindWidget, bInherited=true, widget=null}
  TutorialHighlight {widgetClassPath=&TutorialHighlight}
 ]
}
