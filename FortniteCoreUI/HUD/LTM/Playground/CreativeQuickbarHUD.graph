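# HUD widget that hosts the Creative primary quickbar and keeps it in sync with player and input state.
# Construct binds controller, inventory, HUD context and input events; Destruct removes only some of them.
# The bar shows when Creative mode and quickbar are enabled outside mobile touch play, and maximizes only while the Creative quickbar has focus.
# Move-tool equip toggles phone mode and the component's equipped state; HUD visibility refreshes reposition the phone.
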WidgetBlueprint /FortniteCoreUI/HUD/LTM/Playground/CreativeQuickbarHUD
graphs [
 EventGraph {nodes [
   n1 = OnMoveToolEquippedChanged!() -> (OutputDelegate, then, bIsEnabled:bool=~) {is=title K2Node_CustomEvent}
   > EnablePhoneMode!(
    execute
    self:Creative Quickbar Primary Object Reference=Get CreativeQuickbarPrimary() -> (CreativeQuickbarPrimary:Creative Quickbar Primary Object Reference)
    PhoneModeEnabled:bool=@n1.bIsEnabled
   ) -> (then)
   > n2 = Cast To FortPlayerControllerAthena(Object=GetOwningPlayer())
   n3 = Get CreativeQuickbarComponent!(self:Fort Player Controller Gameplay Object Reference=@n2) -> (CreativeQuickbarComponent:Creative Quickbar Component Object Reference)
   > Is Valid(InputObject=@n3)
   > SetIsQuickbarEquipped!(execute, self:Creative Quickbar Component Object Reference=@n3, bIsEquipped:bool=@n1.bIsEnabled) -> (then)
   Event Construct()
   n4 = Create Event()
   n5 = Get FortInventoryContext!() -> (ReturnValue:Fort Inventory Context Object Reference) {is=title K2Node_GetBlueprintContextEx}
   > Assign On Quickbar Slot Focus Changed!(execute:Exec, self:Fort Inventory Context Object Reference=@n5, Delegate:"Delegate (by ref)"=@n4) -> (then:Exec) {is=title K2Node_AssignDelegate}
   > HandleQuickbarSlotFocusSlotChanged!(execute, self, Quickbar Index:EFortQuickBars Enum=Primary, Slot:int=0) -> (then)
   > SetQuickbarSizes()
   > Bind Event to On Move Tool Equipped Changed(
    self:Creative HUDLContext Object Reference=Get CreativeHUDLContext!() -> (ReturnValue:Creative HUDLContext Object Reference) {is=title K2Node_GetBlueprintContextEx}
    Delegate=Create Event()
   ) {is=title K2Node_AddDelegate}
   > n6 = Cast To FortPlayerControllerAthena(Object=GetOwningPlayer())
   > Set IsCreativeModeEnabled(IsCreativeModeEnabled:bool=IsCreativeModeEnabled!(self:Fort Player Controller Gameplay Object Reference=@n6) -> (ReturnValue:bool)) -> (Output_Get:bool)
   > Set IsCreativeQuickbarEnabled(IsCreativeQuickbarEnabled:bool=IsCreativeQuickbarEnabled!(self:Fort Player Controller Gameplay Object Reference=@n6) -> (ReturnValue:bool)) -> (Output_Get:bool)
   > SetupBarVisibility()
   > Bind Event to On Creative Mode Enabled Changed Delegate(self:Fort Player Controller Gameplay Object Reference=@n6, Delegate=Create Event()) {is=title K2Node_AddDelegate}
   > Bind Event to On Creative Quickbar Enabled Changed Delegate(self:Fort Player Controller Gameplay Object Reference=@n6, Delegate=Create Event()) {is=title K2Node_AddDelegate}
   n7 = Create Event()
   > Bind Event to On HUDElement Visibility Refreshed Delegate(self:Fort Player Controller Gameplay Object Reference=@n6, Delegate=@n7) {is=title K2Node_AddDelegate}
   n8 = Create Event()
   > Bind Event to On Input Method Changed(
    self=Get CommonInputSubsystem!(PlayerController:Player Controller Object Reference=@n6) -> (ReturnValue:Common Input Object Reference) {is=title K2Node_GetSubsystemFromPC}
    Delegate=@n8
   )
   Event Destruct()
   > Unbind Event from On Quickbar Slot Focus Changed(self:Fort Inventory Context Object Reference=@n5, Delegate=@n4) {is=title K2Node_RemoveDelegate}
   > n9 = Cast To FortPlayerControllerAthena(Object=GetOwningPlayer())
   > Unbind Event from On HUDElement Visibility Refreshed Delegate(self:Fort Player Controller Gameplay Object Reference=@n9, Delegate=@n7) {is=title K2Node_RemoveDelegate}
   > Unbind Event from On Input Method Changed(self=Get CommonInputSubsystem(PlayerController=@n9), Delegate=@n8)
  ]}
 HandleQuickbarSlotFocusSlotChanged {nodes [
   n1 = HandleQuickbarSlotFocusSlotChanged!() -> (then, Quickbar Index:EFortQuickBars Enum=Primary, Slot:int) {is=title K2Node_FunctionEntry}
   > Branch(Condition="Not Equal (Enum)"!(A:EFortQuickBars Enum=@n1, B:EFortQuickBars Enum=Get Focused Quickbar() -> (Focused Quickbar:EFortQuickBars Enum=Primary)) -> (ReturnValue:bool))
   > Set Focused Quickbar(Focused Quickbar:EFortQuickBars Enum=@n1) -> (Output_Get:EFortQuickBars Enum=Primary)
   > SetQuickbarSizes()
  ]}
 OnCreativeModeEnabledChanged {nodes [
   n1 = OnCreativeModeEnabledChanged!() -> (then, IsEnabled:bool=~) {is=title K2Node_FunctionEntry}
   > Set IsCreativeModeEnabled(IsCreativeModeEnabled=@n1)
   > SetupBarVisibility()
  ]}
 OnCreativeQuickbarEnabledChanged {nodes [
   n1 = OnCreativeQuickbarEnabledChanged!() -> (then, IsEnabled:bool=~) {is=title K2Node_FunctionEntry}
   > Set IsCreativeQuickbarEnabled(IsCreativeQuickbarEnabled=@n1)
   > SetupBarVisibility()
  ]}
 OnHUDElementVisibilityChanged {nodes [
   OnHUDElementVisibilityChanged() {is=title K2Node_FunctionEntry}
   > SetPhonePosition(self:Creative Quickbar Primary Object Reference=Get CreativeQuickbarPrimary())
  ]}
 OnInputMethodChanged {nodes [
   OnInputMethodChanged!() -> (then, bNewInputType:ECommonInputType Enum) {is=title K2Node_FunctionEntry}
   > SetupBarVisibility()
  ]}
 SetQuickbarSizes {nodes [
   SetQuickbarSizes() {is=title K2Node_FunctionEntry}
   > n1 = Switch on EFortQuickBars!(execute, Selection:EFortQuickBars Enum=Get Focused Quickbar() -> (Focused Quickbar=Primary)) -> (Primary:Exec, Secondary:Exec, Creative:Exec, Max_None:Exec) {is=title K2Node_SwitchEnum}
   n2 = Maximize(execute=@n1.Creative, self:Fort Quickbar Base Object Reference=Get CreativeQuickbarPrimary())
   Return Node(execute=[@n2.then, @n3.then])
   n3 = Minimize(execute=[@n1.Max_None, @n1.Primary, @n1.Secondary], self:Fort Quickbar Base Object Reference=Get CreativeQuickbarPrimary())
  ]}
 SetupBarVisibility {nodes [
   SetupBarVisibility() {is=title K2Node_FunctionEntry}
   n1 = AND Boolean(A=IsMobileGame(), B=IsUsingTouch(PlayerController=GetOwningPlayer()))
   > n2 = Branch(Condition=AND Boolean!(
     A=Get IsCreativeModeEnabled() -> (IsCreativeModeEnabled:bool)
     B=Get IsCreativeQuickbarEnabled() -> (IsCreativeQuickbarEnabled:bool)
     C:bool=NOT Boolean(A=@n1)
    ) -> (ReturnValue))
   n3 = Get CreativeQuickbarPrimary()
   > SetVisibility(self=@n3, InVisibility=SelfHitTestInvisible)
   SetVisibility(execute=@n2.else, self=@n3, InVisibility=Collapsed)
  ]}
]
variables [
 Focused Quickbar {type=EFortQuickBars, default=Primary}
 IsCreativeModeEnabled {type=boolean}
 IsCreativeQuickbarEnabled {type=boolean}
]
widgets {
 info {
  inheritedWidgetCount=0
  parentClass=&UserWidget
  rootWidgetClass=/FortniteCoreUI/HUD/CreativeQuickbarPrimary.CreativeQuickbarPrimary_C
  widgetCount=1
 }
 widgets [CreativeQuickbarPrimary {
   widgetClassPath=/FortniteCoreUI/HUD/CreativeQuickbarPrimary
   bIsVariable=true
   widget=CreativeQuickbarPrimary
   properties {
    elementTag {tagName=HUD.Quickbar.Creative}
    hUDElementTag {gameplayTags [
      {tagName=HUD.Athena.HideOnBus}
      {tagName=Athena.HideForCapture}
      {tagName=HUD.Athena.HideForGadget}
      {tagName=HUD.Athena.HideOnSpectate}
      {tagName=HUD.Athena.Hide.WithInventory}
      {tagName=HUD.Athena.Hide.WithRCPawn}
      {tagName=HUD.Quickbar.Combat}
      {tagName=Athena.Creative.HideForLoading}
      {tagName=HUD.Hide.WhileLoading}
     ]}
   }
  }]
}
