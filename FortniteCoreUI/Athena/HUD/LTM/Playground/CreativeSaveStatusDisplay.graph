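# HUD widget for Fortnite Creative that shows a throbber with "Auto Saving..." or "Saving..." during island saves.
# On plot volume change, it hides and rebinds to the new volume's FortLevelSaveComponent save and backup-save delegates.
# A switcher selects the checkpoint view for in-progress backup saves, which takes priority over autosave.
# Looping timers keep the indicator up briefly, collapsing it once the relevant save is inactive.
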WidgetBlueprint /FortniteCoreUI/Athena/HUD/LTM/Playground/CreativeSaveStatusDisplay
graphs [EventGraph {nodes [
   n1 = OnSaveStateChanged!() -> (OutputDelegate, then, SaveActive:bool=~) {is=title K2Node_CustomEvent}
   > n2 = Set bSaveActive(bSaveActive:bool=@n1.SaveActive) -> (Output_Get:bool)
   n3 = Is Timer Active by Handle!(Handle:Timer Handle Structure=Get DisplayActiveForCheckpointHandle() -> (DisplayActiveForCheckpointHandle:Timer Handle Structure)) -> (ReturnValue:bool)
   > Branch(Condition=AND Boolean(A=@n2, B=NOT Boolean(A=@n3)))
   > SetVisibility(InVisibility=SelfHitTestInvisible)
   > n4 = Set Timer by Event(Delegate=Create Event(), Time=3, bLooping=true)
   > Set DisplayTimerHandle(DisplayTimerHandle:Timer Handle Structure=@n4) -> (Output_Get:Timer Handle Structure)
   n5 = OnVolumeChanged!() -> (OutputDelegate, then, Volume:Fort Volume Object Reference) {is=title K2Node_CustomEvent}
   > SetVisibility(InVisibility=Collapsed)
   > n6 = Is Valid(InputObject=@n5.Volume)
   n7 = Get CurSaveComponent() -> (CurSaveComponent:Fort Level Save Component Object Reference)
   > n8 = Is Valid(InputObject=@n7)
   > Unbind Event from On Client Save State Changed(self:Fort Level Save Component Object Reference=@n7, Delegate=Create Event()) {is=title K2Node_RemoveDelegate}
   > n9 = Unbind Event from On Client Backup Save State Changed(self:Fort Level Save Component Object Reference=@n7, Delegate=Create Event()) {is=title K2Node_RemoveDelegate}
   n10 = GetComponentByClass!(self:Actor Object Reference=@n5.Volume, ComponentClass:Actor Component Class Reference=/Script/FortniteGame.FortLevelSaveComponent) -> (ReturnValue:Fort Level Save Component Object Reference)
   n11 = Is Valid(exec=[@n9.then, @n8.Is Not Valid], InputObject=@n10)
   Set CurSaveComponent(execute=[@n11.Is Not Valid, @n6.Is Not Valid], CurSaveComponent:Fort Level Save Component Object Reference) -> (Output_Get:Fort Level Save Component Object Reference)
   n12 = Set CurSaveComponent(execute=@n11.Is Valid, CurSaveComponent=@n10)
   > Bind Event to On Client Save State Changed(self:Fort Level Save Component Object Reference=@n12, Delegate=Create Event()) {is=title K2Node_AddDelegate}
   > Bind Event to On Client Backup Save State Changed(self:Fort Level Save Component Object Reference=@n12, Delegate=Create Event()) {is=title K2Node_AddDelegate}
   HideSaveAfterDelay_checkpoint() {is=title K2Node_CustomEvent}
   > n13 = Branch(Condition=Get CheckpointActive() -> (CheckpointActive:bool))
   Clear and Invalidate Timer by Handle(execute=@n13.else, Handle=Get DisplayActiveForCheckpointHandle())
   > SetVisibility(InVisibility=Collapsed)
   > SetActiveWidgetIndex(self=Get Switcher() -> (Switcher:Widget Switcher Object Reference), Index=0)
   n14 = OnCheckpointStateChanged!() -> (OutputDelegate, then, SaveState:EBackupSaveState Enum=Ready) {is=title K2Node_CustomEvent}
   > n15 = Set CheckpointActive(CheckpointActive:bool="Equal (Enum)"(A:EBackupSaveState Enum=@n14.SaveState, B:EBackupSaveState Enum=InProgress)) -> (Output_Get:bool)
   > Branch(Condition=@n15)
   > SetVisibility(InVisibility=SelfHitTestInvisible)
   > SetActiveWidgetIndex(self=Get Switcher(), Index=1)
   > Clear and Invalidate Timer by Handle(Handle=Get DisplayTimerHandle() -> (DisplayTimerHandle:Timer Handle Structure))
   > n16 = Set Timer by Event(Delegate=Create Event(), Time=3, bLooping=true)
   > Set DisplayActiveForCheckpointHandle(DisplayActiveForCheckpointHandle:Timer Handle Structure=@n16) -> (Output_Get:Timer Handle Structure)
   HideSaveStatusAfterDelay() {is=title K2Node_CustomEvent}
   n17 = Get DisplayActiveForCheckpointHandle()
   > n18 = Branch(Condition=OR Boolean(A=Get bSaveActive() -> (bSaveActive:bool), B=Is Timer Active by Handle(Handle=@n17)))
   Clear and Invalidate Timer by Handle(execute=@n18.else, Handle=Get DisplayTimerHandle())
   > SetVisibility(InVisibility=Collapsed)
   Event Construct()
   > n19 = Cast To FortPlayerControllerAthena(Object=GetOwningPlayer())
   > Bind Event to On Creative Plot Linked Volume Changed(self:Fort Player Controller Athena Object Reference=@n19, Delegate=Create Event()) {is=title K2Node_AddDelegate}
  ]}]
variables [
 bSaveActive {type=boolean}
 CheckpointActive {type=boolean}
 CurSaveComponent {type=FortLevelSaveComponent, default=null}
 DisplayActiveForCheckpointHandle {type=TimerHandle}
 DisplayTimerHandle {type=TimerHandle}
]
widgets {
 info {inheritedWidgetCount=0, parentClass=&UserWidget, rootWidgetClass=&WidgetSwitcher, widgetCount=7}
 widgets [
  Switcher {widgetClassPath=&WidgetSwitcher, bIsVariable=true, widget=Switcher}
  Normal {widgetClassPath=&VerticalBox, parent=Switcher, widget=Normal, slot {class=WidgetSwitcherSlot}}
  MatchmakingSpinner {
   widgetClassPath=&Image
   bIsVariable=true
   parent=Normal
   widget=MatchmakingSpinner
   properties {
    brush {
     bIsDynamicallyLoaded=false
     drawAs=Image
     imageType=NoImage
     mirroring=NoMirror
     resourceName=None
     resourceObject=/FortUILibrary/UIKit/Throbber/MI_UIKit_Monochrome_Throbber
     tiling=NoTile
     imageSize=<40 40>
     margin=<ltrb 0 0 0 0>
     outlineSettings {
      bUseBrushTransparency=false
      roundingType=HalfHeightRadius
      width=0
      color {colorUseRule=UseColor_Specified, specifiedColor=<rgba 0 0 0 0>}
      cornerRadii=<0 0 0 0>
     }
     tintColor {colorUseRule=UseColor_Specified, specifiedColor=<rgba 1 1 1 0.6>}
     uVRegion {bIsValid=false, max=<0 0>, min=<0 0>}
    }
    colorAndOpacity=<rgba 1 0.9301 0 1>
   }
   slot {class=VerticalBoxSlot, properties {horizontalAlignment=HAlign_Center, verticalAlignment=VAlign_Center}}
  }
  TextBlock_0 {
   widgetClassPath=&TextBlock
   parent=Normal
   widget=TextBlock_0
   properties {
    text=Auto Saving...
    colorAndOpacity {colorUseRule=UseColor_Specified, specifiedColor=<rgba 1 0.9616 0.3461 0.6>}
    font {
     bForceMonospaced=false
     bMaterialIsStencil=false
     fontMaterial=null
     fontObject=&BurbankSmall
     letterSpacing=0
     monospacedWidth=1
     size=16
     skewAmount=0
     typefaceFontName=Black
     outlineSettings {
      bApplyOutlineToDropShadows=false
      bMiteredCorners=false
      bSeparateFillAlpha=false
      outlineMaterial=null
      outlineSize=0
      outlineColor=<rgba 0 0 0 1>
     }
    }
   }
   slot {class=VerticalBoxSlot, properties {padding=<ltrb 0 4 0 0>}}
  }
  Checkpoint {widgetClassPath=&VerticalBox, parent=Switcher, widget=Checkpoint, slot {class=WidgetSwitcherSlot}}
  Image_0 {
   widgetClassPath=&Image
   bIsVariable=true
   parent=Checkpoint
   widget=Image_0
   properties {
    brush {
     bIsDynamicallyLoaded=false
     drawAs=Image
     imageType=NoImage
     mirroring=NoMirror
     resourceName=None
     resourceObject=&MI_UIKit_Monochrome_Throbber
     tiling=NoTile
     imageSize=<40 40>
     margin=<ltrb 0 0 0 0>
     outlineSettings {
      bUseBrushTransparency=false
      roundingType=HalfHeightRadius
      width=0
      color {colorUseRule=UseColor_Specified, specifiedColor=<rgba 0 0 0 0>}
      cornerRadii=<0 0 0 0>
     }
     tintColor {colorUseRule=UseColor_Specified, specifiedColor=<rgba 0.9366 1 0.99 0.9>}
     uVRegion {bIsValid=false, max=<0 0>, min=<0 0>}
    }
    colorAndOpacity=<rgba 0.9263 1 0.9448 1>
   }
   slot {class=VerticalBoxSlot, properties {horizontalAlignment=HAlign_Center, verticalAlignment=VAlign_Center}}
  }
  TextBlock_1 {
   widgetClassPath=&TextBlock
   parent=Checkpoint
   widget=TextBlock_1
   properties {
    text=Saving...
    colorAndOpacity {colorUseRule=UseColor_Specified, specifiedColor=<rgba 1 0.9963 0.9853 0.9>}
    font {
     bForceMonospaced=false
     bMaterialIsStencil=false
     fontMaterial=null
     fontObject=&BurbankSmall
     letterSpacing=0
     monospacedWidth=1
     size=16
     skewAmount=0
     typefaceFontName=Black
     outlineSettings {
      bApplyOutlineToDropShadows=false
      bMiteredCorners=false
      bSeparateFillAlpha=false
      outlineMaterial=null
      outlineSize=0
      outlineColor=<rgba 0 0 0 1>
     }
    }
   }
   slot {class=VerticalBoxSlot, properties {padding=<ltrb 0 4 0 0>}}
  }
 ]
}
