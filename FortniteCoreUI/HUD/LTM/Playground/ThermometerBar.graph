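# Playground LTM thermometer HUD bar widget showing progress, current/max values and a cost preview readout through a PreviewBar_Memory child.
# Setters update the texts, the hidden fill bar and a dynamic segmented material on the preview image.
# Preview value plus progress reaching 1 turns on the material's warning color and, for a plus change, red preview text.
# A WasFull flag swaps fill colors only when current crosses max.
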WidgetBlueprint /FortniteCoreUI/HUD/LTM/Playground/ThermometerBar
graphs [
 Active previewText {nodes [
   n1 = Active previewText!() -> (then, PreviewTextNUM int:int, Plus or minus:bool=~) {is=title K2Node_FunctionEntry}
   > Set previewText!(
    execute
    self:Preview Bar Memory Object Reference=Get PreviewBar_Memory() -> (PreviewBar_Memory:Preview Bar Memory Object Reference)
    PreviewTextNUM:double=0
    PreviewTextNUM int:int=@n1
    Plus or minus:bool=@n1.Plus or minus
   ) -> (then) {is=title K2Node_CallFunction}
   n2 = Get ValueProgress() -> (ValueProgress:double)
   n3 = Get PreviewValue() -> (PreviewValue:double)
   n4 = Branch(Condition="float >= float"(A=float + float(A=@n3, B=@n2) {is=title K2Node_CommutativeAssociativeBinaryOperator}, B=1) {is=title K2Node_CallFunction})
   > n5 = Branch(Condition=@n1.Plus or minus)
   n6 = Get PreviewBar_Memory()
   > setTextColorWhite(self:Preview Bar Memory Object Reference=@n6)
   setTextColorred(execute=@n5.else, self:Preview Bar Memory Object Reference=@n6)
   setTextColorWhite(execute=@n4.else, self=@n6)
  ]}
 EventGraph {nodes [
   Event PreConstruct()
   > InitializeBar()
  ]}
 InitializeBar {nodes [
   SetStyle!(
    execute
    self:Common Text Object Reference=Get CurrentValue() -> (CurrentValue:Common Text Object Reference)
    InStyle:Common Text Style Class Reference=/FortUILibrary/Styles/Typography/Foundation/TextStyle-Base-XS-B.TextStyle-Base-XS-B_C
   ) -> (then)
   > n1 = SetStyle(
    self=Get MaxValue() -> (MaxValue:Common Text Object Reference)
    InStyle=/FortUILibrary/Styles/Typography/Foundation/TextStyle-Base-XS.TextStyle-Base-XS_C
   )
   SetPadding!(
    execute=[@n4.then, @n1.then]
    self:Overlay Slot Object Reference=SlotAsOverlaySlot(Widget=Get TextArea() -> (TextArea:Horizontal Box Object Reference))
    InPadding:Margin Structure=Make Margin(Left=16, Top=0, Right=0, Bottom=0)
   ) -> (then)
   InitializeBar() {is=title K2Node_FunctionEntry}
   > n2 = GetDynamicMaterial(self=Get ProgressBarPreview() -> (ProgressBarPreview:Image Object Reference))
   > n3 = Set MIpreviewbar(MIpreviewbar:Material Instance Dynamic Object Reference=@n2) -> (Output_Get:Material Instance Dynamic Object Reference)
   > SetScalarParameterValue(self=@n3, ParameterName=twoColorBlend, Value=1)
   > SetColorA!(
    execute
    self:Fort Simple Material Progress Bar Object Reference=Get ProgressBarFill() -> (ProgressBarFill:Simple Material Progress Bar Object Reference)
    ColorA:Linear Color Structure=Get ColorA() -> (ColorA:Linear Color Structure)
   ) -> (then)
   > n4 = SetColorB!(
    execute
    self:Fort Simple Material Progress Bar Object Reference=Get ProgressBarFill()
    ColorB:Linear Color Structure=Get ColorB() -> (ColorB:Linear Color Structure)
   ) -> (then)
  ]}
 SetCostPreviewBarVisibility {nodes [
   n1 = SetCostPreviewBarVisibility!() -> (then, Visibility:bool=~) {is=title K2Node_FunctionEntry}
   > n2 = Branch(Condition=@n1)
   n3 = Get PreviewBar_Memory()
   > SetVisibility(self=@n3, InVisibility=Visible)
   > SetRenderOpacity(self=@n3, InOpacity=1)
   SetVisibility(execute=@n2.else, self=@n3, InVisibility=Hidden)
   > SetRenderOpacity(self=@n3, InOpacity=0)
  ]}
 SetCurrentValue {nodes [
   n1 = SetCurrentValue!() -> (then, Value:double=~) {is=title K2Node_FunctionEntry}
   > Branch(Condition="Not Equal (Float)"!(A:double=@n1, B:double=Get ValueCurrent() -> (ValueCurrent:double)) -> (ReturnValue:bool) {is=title K2Node_CallFunction})
   > Set ValueCurrent(ValueCurrent:double=@n1) -> (Output_Get:double)
   > "SetText (Text)"(
    self=Get CurrentValue()
    InText="To Text (Float)"(
     Value=Get ValueCurrent()
     RoundingMode=FromZero
     MinimumIntegralDigits=1
     MaximumIntegralDigits=324
     MinimumFractionalDigits=0
     MaximumFractionalDigits=0
    )
   )
  ]}
 SetMaxValue {nodes [
   n1 = SetMaxValue!() -> (then, Value:double=~) {is=title K2Node_FunctionEntry}
   > Branch(Condition="Not Equal (Float)"(A=@n1, B=Get ValueMax() -> (ValueMax:double)))
   > Set ValueMax(ValueMax:double=@n1) -> (Output_Get:double)
   > "SetText (Text)"(
    self=Get MaxValue()
    InText="To Text (Float)"(
     Value=Get ValueMax()
     RoundingMode=FromZero
     MinimumIntegralDigits=1
     MaximumIntegralDigits=324
     MinimumFractionalDigits=0
     MaximumFractionalDigits=0
    )
   )
  ]}
 SetPreview Value {nodes [
   n1 = SetPreview Value!() -> (then, Preview value:double=~) {is=title K2Node_FunctionEntry}
   > Set PreviewValue(PreviewValue:double=@n1) -> (Output_Get:double)
   > SetScalarParameterValue(self=Get MIpreviewbar() -> (MIpreviewbar:Material Instance Dynamic Object Reference), ParameterName=delayProgress, Value=@n1)
   n2 = Get ValueProgress()
   > n3 = Branch(Condition="float >= float"(A=float + float(A=@n1, B=@n2), B=1))
   n4 = Get MIpreviewbar()
   > SetScalarParameterValue(self=@n4, ParameterName=warningColor, Value=1)
   SetScalarParameterValue(execute=@n3.else, self=@n4, ParameterName=warningColor, Value=0)
  ]}
 SetValueProgress {nodes [
   n1 = SetValueProgress!() -> (then, Value:double=~) {is=title K2Node_FunctionEntry}
   n2 = "Clamp (Float)"(Value=@n1, Min=0, Max=@n1)
   > SetProgress!(execute, self:Fort Simple Material Progress Bar Object Reference=Get ProgressBarFill(), Progress:float=@n2) -> (then)
   > Set ValueProgress(ValueProgress:double=@n2) -> (Output_Get:double)
   > SetScalarParameterValue(self=Get MIpreviewbar(), ParameterName=Progress, Value=@n2)
   > n3 = Branch(Condition="float >= float"(A=Get ValueCurrent(), B=Get ValueMax()))
   n4 = Get WasFull() -> (WasFull:bool)
   > n5 = Branch(Condition=@n4)
   n6 = Get ProgressBarFill()
   n7 = Get ColorB()
   SetColorA(execute=@n5.else, self=@n6, ColorA=@n7)
   > n8 = SetColorB(self=@n6, ColorB=@n7)
   Set WasFull(execute=[@n9.then, @n8.then], WasFull:bool=NOT Boolean(A=Get WasFull())) -> (Output_Get:bool)
   Branch(execute=@n3.else, Condition=@n4)
   > SetColorA(self=@n6, ColorA=Get ColorA())
   > n9 = SetColorB(self=@n6, ColorB=@n7)
  ]}
 set TextColorRed {nodes [
   set TextColorRed() {is=title K2Node_FunctionEntry}
   > setTextColorred(self=Get PreviewBar_Memory())
  ]}
 set TextWhite {nodes [
   set TextWhite() {is=title K2Node_FunctionEntry}
   > setTextColorWhite(self=Get PreviewBar_Memory())
  ]}
]
variables [
 ColorA {type=LinearColor, default=<rgba 0.6823 0.6329 0.0022 1>}
 ColorB {type=LinearColor, default=<rgba 0.6927 0 0.0127 1>}
 MIpreviewbar {type=MaterialInstanceDynamic}
 PreviewValue {type=number}
 ValueCurrent {type=number, default=-1}
 ValueMax {type=number, default=-1}
 ValueProgress {type=number}
 WasFull {type=boolean}
]
widgets {
 info {inheritedWidgetCount=0, parentClass=&UserWidget, rootWidgetClass=&HorizontalBox, widgetCount=10}
 widgets [
  HorizontalBox_66 {widgetClassPath=&HorizontalBox, widget=HorizontalBox_66}
  BarSizeBox {
   widgetClassPath=&SizeBox
   bIsVariable=true
   parent=HorizontalBox_66
   widget=BarSizeBox
   properties {bOverride_HeightOverride=true, bOverride_WidthOverride=true, heightOverride=24, widthOverride=600}
   slot {class=HorizontalBoxSlot, properties {horizontalAlignment=HAlign_Right, verticalAlignment=VAlign_Top, size {sizeRule=Fill, value=1}}}
  }
  "[Overlay]" {
   widgetClassPath=&Overlay
   parent=BarSizeBox
   widget="[Overlay]"
   properties {visibility=HitTestInvisible}
   slot {class=SizeBoxSlot}
  }
  ProgressBarFill {
   widgetClassPath=/Game/Athena/HUD/SimpleMaterialProgressBar
   bIsVariable=true
   parent="[Overlay]"
   widget=ProgressBarFill
   properties {
    bOverrideDefaultColorA=true
    bOverrideDefaultColorB=true
    bOverrideDefaultColorBackground=true
    visibility=Hidden
    cachedColorA=<rgba 0.6823 0.6329 0.0022 1>
    cachedColorB=<rgba 0.6927 0 0.0127 1>
    cachedColorBackground=<rgba 0 0 0 0.498>
   }
   slot {class=OverlaySlot, properties {horizontalAlignment=HAlign_Fill, verticalAlignment=VAlign_Fill}}
  }
  ProgressBarPreview {
   widgetClassPath=&Image
   bIsVariable=true
   parent="[Overlay]"
   widget=ProgressBarPreview
   properties {brush {
     bIsDynamicallyLoaded=false
     drawAs=Image
     imageType=NoImage
     mirroring=NoMirror
     resourceName=None
     resourceObject=/FortniteCoreUI/Athena/HUD/LTM/Playground/Thermometer/M_ProgressBar_Segmented_Inst
     tiling=NoTile
     imageSize=<32 32>
     margin=<ltrb 0 0 0 0>
     outlineSettings {
      bUseBrushTransparency=false
      roundingType=HalfHeightRadius
      width=0
      color {colorUseRule=UseColor_Specified, specifiedColor=<rgba 0 0 0 0>}
      cornerRadii=<0 0 0 0>
     }
     tintColor {colorUseRule=UseColor_Specified, specifiedColor=<rgba 1 1 1 1>}
     uVRegion {bIsValid=false, max=<0 0>, min=<0 0>}
    }}
   slot {class=OverlaySlot, properties {horizontalAlignment=HAlign_Fill, verticalAlignment=VAlign_Fill}}
  }
  TextArea {
   widgetClassPath=&HorizontalBox
   bIsVariable=true
   parent="[Overlay]"
   widget=TextArea
   properties {visibility=HitTestInvisible}
   slot {class=OverlaySlot, properties {verticalAlignment=VAlign_Center, padding=<ltrb 16 0 0 0>}}
  }
  CurrentValue {
   widgetClassPath=&CommonTextBlock
   bIsVariable=true
   parent=TextArea
   widget=CurrentValue
   properties {
    justification=Right
    minDesiredWidth=0.1524
    text=999?
    font {
     bForceMonospaced=false
     bMaterialIsStencil=false
     fontMaterial=null
     fontObject=&NotoSans
     letterSpacing=0
     monospacedWidth=1
     size=16
     skewAmount=0
     typefaceFontName=NotoSans-Bold
     outlineSettings {
      bApplyOutlineToDropShadows=false
      bMiteredCorners=false
      bSeparateFillAlpha=false
      outlineMaterial=null
      outlineSize=1
      outlineColor=<rgba 0 0 0 0.25>
     }
    }
    renderTransform {angle=0, scale=<1 1>, shear=<0 0>, translation=<0 1>}
    shadowColorAndOpacity=<rgba 0 0 0 0.35>
    shadowOffset=<1.5 0.25>
   }
   slot {class=HorizontalBoxSlot, properties {verticalAlignment=VAlign_Center, padding=<ltrb 0 -2.5 0 0>, size {sizeRule=Fill, value=1}}}
  }
  Split {
   widgetClassPath=&Image
   parent=TextArea
   widget=Split
   properties {
    brush {
     bIsDynamicallyLoaded=false
     drawAs=Image
     imageType=NoImage
     mirroring=NoMirror
     resourceName=None
     resourceObject=null
     tiling=NoTile
     imageSize=<4 32>
     margin=<ltrb 0 0 0 0>
     outlineSettings {
      bUseBrushTransparency=false
      roundingType=HalfHeightRadius
      width=0
      color {colorUseRule=UseColor_Specified, specifiedColor=<rgba 0 0 0 0>}
      cornerRadii=<0 0 0 0>
     }
     tintColor {colorUseRule=UseColor_Specified, specifiedColor=<rgba 1 1 1 1>}
     uVRegion {bIsValid=false, max=<0 0>, min=<0 0>}
    }
    colorAndOpacity=<rgba 1 1 1 0.2>
   }
   slot {class=HorizontalBoxSlot, properties {padding=<ltrb 4 2 4 2>}}
  }
  MaxValue {
   widgetClassPath=&CommonTextBlock
   bIsVariable=true
   parent=TextArea
   widget=MaxValue
   properties {
    text=999?
    font {
     bForceMonospaced=false
     bMaterialIsStencil=false
     fontMaterial=null
     fontObject=&NotoSans
     letterSpacing=0
     monospacedWidth=1
     size=16
     skewAmount=0
     typefaceFontName=NotoSans-Bold
     outlineSettings {
      bApplyOutlineToDropShadows=false
      bMiteredCorners=false
      bSeparateFillAlpha=false
      outlineMaterial=null
      outlineSize=1
      outlineColor=<rgba 0 0 0 0.25>
     }
    }
    renderTransform {angle=0, scale=<1 1>, shear=<0 0>, translation=<0 1>}
   }
   slot {class=HorizontalBoxSlot, properties {horizontalAlignment=HAlign_Left, verticalAlignment=VAlign_Center, padding=<ltrb 0 -2.5 0 0>}}
  }
  PreviewBar_Memory {
   widgetClassPath=/FortniteCoreUI/HUD/LTM/Playground/Thermometer/PreviewBar_Memory
   bIsVariable=true
   parent=HorizontalBox_66
   widget=PreviewBar_Memory
   properties {renderOpacity=0, visibility=Visible}
   slot {class=HorizontalBoxSlot}
  }
 ]
}
