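# Legacy Creative/Playground performance HUD widget showing memory used against budget on a thermometer bar, alongside the save status display, positioned per platform.
# SetUIMetrics feeds UpdateUIMetrics_Original, which scales usage to the bar and loops a red title animation when full.
# Selection cost refreshes add a signed preview delta, turning red when a positive cost reaches the budget.
# An active heatmap collapses the HUD on metric updates and skips cost previews.
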WidgetBlueprint /FortniteCoreUI/HUD/LTM/Playground/PerformanceHUD_Legacy
animations [BarFull_Server {
  binding_count=1
  length_seconds=0.773
  object=&self:BarFull_Server
  sequence {bindings [ThermometerTitle_Server {
     id=#1
     display_name=ThermometerTitle_GlobalServer
     possessed_class=CommonTextBlock
     tracks [{
       type=MovieSceneColorTrack
       display_name=Color and Opacity
       sections [{
         type=MovieSceneColorSection
         end_frame=15
         start_frame=0
         channels [
          Color.R {type=MovieSceneScriptingFloatChannel, keys [[0,1],[15.324,0.6939]]}
          Color.G {type=MovieSceneScriptingFloatChannel, keys [[0,1],[15.324,0]]}
          Color.B {type=MovieSceneScriptingFloatChannel, keys [[0,1],[15.324,0.0123]]}
          Color.A {type=MovieSceneScriptingFloatChannel, const [1,2]}
         ]
        }]
      }]
    }]}
 }]
graphs [
 EventGraph {nodes [
   Event Construct()
   > SetVisibility(
    self=[
     Get ThermometerArea_GlobalMemory() -> (ThermometerArea_GlobalMemory:Scale Box Object Reference)
     Get ThermometerArea_Preview() -> (ThermometerArea_Preview:Scale Box Object Reference)
    ]
    InVisibility=Select(Option 0=Collapsed, Option 1=HitTestInvisible, Index=true)
   )
   n1 = Get ThermometerTitle_GlobalServer() -> (ThermometerTitle_GlobalServer:Common Text Object Reference)
   > "SetText (Text)"(self=@n1, InText="LOCTEXT(\"MEMORY USED\")")
   > SetRenderTransform!(
    execute
    self:Widget Object Reference=@n1
    InTransform:Widget Transform Structure=Make Widget Transform!(Translation:Vector 2D Structure=(X=-30,Y=0), Scale:Vector 2D Structure=(X=1,Y=1), Shear:Vector 2D Structure=(X=0,Y=0), Angle:float=0) -> (WidgetTransform:Widget Transform Structure) {is=title K2Node_MakeStruct}
   ) -> (then)
   n2 = Get ThermometerBar_GlobalServer() -> (ThermometerBar_GlobalServer:Thermometer Bar Object Reference)
   > SetCurrentValue!(execute, self:Thermometer Bar Object Reference=@n2, Value:double=0) -> (then)
   > SetMaxValue!(execute, self:Thermometer Bar Object Reference=@n2, Value:double=Get MaxServerBarValue() -> (MaxServerBarValue:double)) -> (then)
   > SetValueProgress!(execute, self:Thermometer Bar Object Reference=@n2, Value:double=0) -> (then)
   > SetVisibility(self=@n2, InVisibility=Hidden)
   > n3 = Sequence()
   > SetWidgetPositionOnPlatforms!(
    execute
    self
    isApplyToBattleLab:bool=true
    MobilePos:Vector 2D Structure=(X=0,Y=100)
    DefaultPos:Vector 2D Structure=(X=0,Y=65)
    WidgetToRepos:Widget Object Reference=Get ThermometerArea_GlobalMemory()
   ) -> (then)
   SetWidgetPositionOnPlatforms(
    execute=@n3.then_1
    isApplyToBattleLab=false
    MobilePos=(X=-30,Y=175)
    DefaultPos=(X=-320,Y=50)
    WidgetToRepos=Get CreativeSaveStatusDisplay() -> (CreativeSaveStatusDisplay:Creative Save Status Display Object Reference)
   )
   Event HideUIMetrics() {is=title K2Node_Event}
   > Set CurrentMemoryValue(CurrentMemoryValue:int=0) -> (Output_Get:int=0)
   > Set CurrentMemoryValueAfterRound(CurrentMemoryValueAfterRound:int=0) -> (Output_Get:int=0)
   > SetVisibility(InVisibility=Hidden)
   n4 = Event SetUIMetrics!() -> (OutputDelegate, then, VolumePerformanceMetrics:"Volume Performance Metrics Structure (by ref)") {is=title K2Node_Event}
   > n5 = Branch(Condition=NOT Boolean(A=IsHeatmapEnabled!(self:Creative Performance HUDBase Object Reference) -> (ReturnValue:bool)))
   > SetVisibility(InVisibility=HitTestInvisible)
   n6 = Break Volume Performance Metrics!(VolumePerformanceMetrics:"Volume Performance Metrics Structure (by ref)"=@n4.VolumePerformanceMetrics) -> (
    PerformanceValue:int
    PerformanceMaxValue:int
    PerformanceValueMargin:int
    PerformanceValueCellIndex:Int Point Structure
    PerformanceLowendThreshold:int
    PreviewDeltaValue:int
    Category:EFortBudgetCategory Enum
    PlayerID:int
    MostExpensivePerformanceValue:int
    MostExpensivePreviewDeltaValue:int
    GlobalMemoryValue:int
    MostExpensivePerformanceValueCellIndex:Int Point Structure
   ) {is=title K2Node_BreakStruct}
   > Set MaxMemoryBudgetValue(MaxMemoryBudgetValue:int=@n6.PerformanceMaxValue) -> (Output_Get:int=0)
   > Set CurrentMemoryValue(CurrentMemoryValue=@n6) -> (Output_Get=0)
   > UpdateUIMetrics_Original!(
    execute
    self
    PerformanceValue:int=@n6
    PerformanceMaxValue:int=@n6.PerformanceMaxValue
    PerformanceMaxBarValue:double=Get MaxServerBarValue()
    PerformanceCategory:EFortBudgetCategory Enum=@n6.Category
    isServerMemory:bool=true
   ) -> (then)
   SetVisibility(execute=@n5.else, InVisibility=Collapsed)
   n7 = Event RefreshCostPreviewOfSelectedActor!() -> (OutputDelegate, then, CostPreviewOfSelection:int=0, isThrowingPlayset:bool) {is=title K2Node_Event}
   > Branch(Condition=NOT Boolean(A=IsHeatmapEnabled()))
   n8 = Get MaxMemoryBudgetValue() -> (MaxMemoryBudgetValue:int=0)
   n9 = float / float(A="To Float (Integer)"(InInt=@n7.CostPreviewOfSelection), B="Max (Float)"(A="To Float (Integer)"(InInt=@n8), B=1)) {is=title K2Node_CallFunction}
   n10 = Get GlobalProgress() -> (GlobalProgress:double)
   > n11 = Branch(Condition="float >= float"(A=float + float(A=@n10, B=@n9) {is=title K2Node_CommutativeAssociativeBinaryOperator}, B=1) {is=title K2Node_CallFunction})
   n12 = "integer < integer"!(A:int=@n7.CostPreviewOfSelection, B:int=0) -> (ReturnValue:bool) {is=title K2Node_CallFunction}
   > n13 = Branch(Condition=@n12)
   n14 = Get PreviewBar_Memory() -> (PreviewBar_Memory:Preview Bar Memory Object Reference)
   > n15 = setTextColorWhite(self:Preview Bar Memory Object Reference=@n14)
   n16 = Get ThermometerBar_GlobalServer()
   SetPreview Value!(execute=[@n24.then, @n15.then, @n25.then], self:Thermometer Bar Object Reference=@n16, Preview value:double=@n9) -> (then)
   > n17 = Branch(Condition=@n12)
   > n18 = SetRenderTransform(self=Get PreviewBar_Memory(), InTransform=Make Widget Transform(Translation=(X=5,Y=33), Scale=(X=1.3,Y=1.3), Shear=(X=0,Y=0), Angle=0))
   n19 = "To Float (Integer)"(InInt=int + int(A=@n7.CostPreviewOfSelection, B=Get CurrentMemoryValue() -> (CurrentMemoryValue:int=0)) -> (ReturnValue=0) {is=title K2Node_CommutativeAssociativeBinaryOperator})
   n20 = "To Float (Integer)"(InInt=Get MaxMemoryBudgetValue() -> (MaxMemoryBudgetValue=0))
   n21 = float * float(A=float / float(A="Clamp (Float)"(Value=@n19, Min=0, Max=@n19), B="Max (Float)"(A=@n20, B=1)), B=Get MaxServerBarValue())
   n22 = "Absolute (Integer)"!(A:int=int - int(A=Round(A=@n21) {is=title K2Node_CallFunction}, B=Get CurrentMemoryValueAfterRound() -> (CurrentMemoryValueAfterRound:int=0)) {is=title K2Node_CallFunction}) -> (ReturnValue:int=0)
   Set previewText!(
    execute=[@n18.then, @n23.then]
    self:Preview Bar Memory Object Reference=Get PreviewBar_Memory()
    PreviewTextNUM:double=0
    PreviewTextNUM int:int=@n22
    Plus or minus:bool=@n12
   ) -> (then) {is=title K2Node_CallFunction}
   > Active previewText!(execute, self:Thermometer Bar Object Reference=@n16, PreviewTextNUM int:int=@n22, Plus or minus:bool=@n12) -> (then)
   n23 = SetRenderTransform(
    execute=@n17.else
    self=Get PreviewBar_Memory()
    InTransform=Make Widget Transform(Translation=(X=0,Y=-2), Scale=(X=1.3,Y=1.3), Shear=(X=0,Y=0), Angle=0)
   )
   n24 = setTextColorred(execute=@n13.else, self:Preview Bar Memory Object Reference=@n14)
   n25 = setTextColorWhite(execute=@n11.else, self=@n14)
  ]}
 Set Costpreview Visibility {nodes [
   n1 = Set Costpreview Visibility!() -> (then, visibility:int) {is=title K2Node_FunctionEntry}
   > n2 = Switch on Int!(execute:Exec, Selection:int=@n1) -> (Default:Exec, 0:Exec, 1:Exec, 2:Exec)
   n3 = Get PreviewBar_Memory()
   SetVisibility(execute=[@n2.0, @n2.Default], self=[Get ThermometerArea_Preview(), @n3], InVisibility=Collapsed)
   > SetCostPreviewBarVisibility!(execute, self:Thermometer Bar Object Reference=Get ThermometerBar_GlobalServer(), Visibility:bool=false) -> (then)
   SetVisibility(execute=@n2.1, self=[Get ThermometerArea_Preview(), @n3], InVisibility=HitTestInvisible)
   > SetCostPreviewBarVisibility(self=Get ThermometerBar_GlobalServer(), Visibility=false)
   SetVisibility(execute=@n2.2, self=[Get ThermometerArea_Preview(), @n3], InVisibility=Collapsed)
   > SetCostPreviewBarVisibility(self=Get ThermometerBar_GlobalServer(), Visibility=true)
  ]}
 SetWidgetPositionOnPlatforms {nodes [
   n1 = SetWidgetPositionOnPlatforms!() -> (then, isApplyToBattleLab:bool=~, MobilePos:Vector 2D Structure, DefaultPos:Vector 2D Structure, WidgetToRepos:Widget Object Reference) {is=title K2Node_FunctionEntry}
   > n2 = Branch(Condition=@n1)
   > n3 = Cast To FortGameStateAthena!(execute, Object=GetGameState!() -> (ReturnValue:Game State Base Object Reference)) -> (then, CastFailed, AsFort Game State Athena:Fort Game State Athena Object Reference) {is=title K2Node_DynamicCast}
   > Is Valid(InputObject=@n3)
   > n4 = Branch(Condition=IsMobileGame())
   n5 = Set TargetPos(execute=[@n4.then, @n8.then], TargetPos:Vector 2D Structure=@n1.MobilePos) -> (Output_Get:Vector 2D Structure)
   n6 = SlotAsCanvasSlot!(Widget:Widget Object Reference=@n1.WidgetToRepos) -> (ReturnValue:Canvas Panel Slot Object Reference)
   Is Valid(exec=[@n7.then, @n5.then], InputObject=@n6)
   > SetPosition!(execute, self:Canvas Panel Slot Object Reference=@n6, InPosition:Vector 2D Structure=Get TargetPos() -> (TargetPos:Vector 2D Structure)) -> (then)
   > Return Node()
   n7 = Set TargetPos(execute=[@n8.else, @n4.else], TargetPos=@n1.DefaultPos)
   n8 = Branch(execute=@n2.else, Condition=IsMobileGame())
  ]}
 UpdateUIMetrics_Original {nodes [
   n1 = UpdateUIMetrics_Original!() -> (
    then
    PerformanceValue:int
    PerformanceMaxValue:int
    PerformanceMaxBarValue:double=~
    PerformanceCategory:EFortBudgetCategory Enum=Memory
    isServerMemory:bool=~
   ) {is=title K2Node_FunctionEntry}
   > Branch(Condition="Equal (Enum)"(A:EFortBudgetCategory Enum=@n1.PerformanceCategory, B:EFortBudgetCategory Enum=Memory))
   > n2 = Branch(Condition=NOT Boolean(A=@n1.isServerMemory))
   n3 = "To Float (Integer)"(InInt=@n1)
   n4 = float / float(A="Clamp (Float)"(Value=@n3, Min=0, Max=@n3), B="Max (Float)"(A="To Float (Integer)"(InInt=@n1.PerformanceMaxValue), B=1))
   n5 = Round(A=float * float(A=@n4, B=@n1.PerformanceMaxBarValue))
   Set CurrentMemoryValueAfterRound(execute=@n2.else, CurrentMemoryValueAfterRound=@n5) -> (Output_Get=0)
   n6 = Get ThermometerBar_GlobalServer()
   > SetCurrentValue(self=@n6, Value="To Float (Integer)"(InInt=@n5))
   > SetMaxValue(self=@n6, Value=@n1.PerformanceMaxBarValue)
   > n7 = Set GlobalProgress(GlobalProgress:double=@n4) -> (Output_Get:double)
   > SetValueProgress(self=@n6, Value=@n7)
   > SetVisibility(self=@n6, InVisibility=HitTestInvisible)
   > Set BarFullAnimation(BarFullAnimation:Widget Animation Object Reference=Get BarFull_Server() -> (BarFull_Server:Widget Animation Object Reference)) -> (Output_Get:Widget Animation Object Reference)
   > n8 = Branch(Condition="float >= float"(A=@n4, B=1))
   > PlayAnimationTimeRange!(
    execute
    self:User Widget Object Reference
    InAnimation:Widget Animation Object Reference=Get BarFullAnimation() -> (BarFullAnimation:Widget Animation Object Reference)
    StartAtTime:float=0
    EndAtTime:float=0
    NumLoopsToPlay:int=0
    PlayMode:EUMGSequencePlayMode Enum=PingPong
    PlaybackSpeed:float=1
    bRestoreState:bool=false
   ) -> (then, ReturnValue:Widget Animation Handle Structure)
   Branch(execute=@n8.else, Condition=IsAnimationPlaying(InAnimation=Get BarFullAnimation()))
   > StopAnimation(InAnimation=Get BarFullAnimation())
   n9 = Get BarFullAnimation()
   > PlayAnimationTimeRange(InAnimation=@n9, StartAtTime=0, EndAtTime=GetEndTime(self=@n9), NumLoopsToPlay=1, PlayMode=Reverse, PlaybackSpeed=1, bRestoreState=false)
  ]}
]
variables [
 CurrentMemoryValue {type=integer}
 CurrentMemoryValueAfterRound {type=integer}
 GlobalProgress {type=number}
 MaxMemoryBudgetValue {type=integer}
 MaxServerBarValue {type=number, default=100000}
]
widgets {
 info {inheritedWidgetCount=1, parentClass=/Script/FortniteUI.CreativePerformanceHUDBase, rootWidgetClass=&CanvasPanel, widgetCount=11}
 widgets [
  CanvasPanel_0 {widgetClassPath=&CanvasPanel, widget=CanvasPanel_0}
  ThermometerArea_GlobalMemory {
   widgetClassPath=&ScaleBox
   bIsVariable=true
   parent=CanvasPanel_0
   widget=ThermometerArea_GlobalMemory
   properties {stretch=UserSpecified, userSpecifiedScale=0.8, visibility=Collapsed}
   slot {
    class=CanvasPanelSlot
    properties {bAutoSize=true, layoutData {alignment=<0.5 0>, anchors {maximum=<0.5 0>, minimum=<0.5 0>}, offsets=<ltrb 0 35 100 30>}}
   }
  }
  Overlay {widgetClassPath=&Overlay, parent=ThermometerArea_GlobalMemory, widget=Overlay, slot {class=ScaleBoxSlot}}
  Overlay_3 {widgetClassPath=&Overlay, parent=Overlay, widget=Overlay_3, slot {class=OverlaySlot}}
  VerticalBox {widgetClassPath=&VerticalBox, parent=Overlay_3, widget=VerticalBox, slot {class=OverlaySlot}}
  ThermometerTitle_GlobalServer {
   widgetClassPath=&CommonTextBlock
   bIsVariable=true
   parent=VerticalBox
   widget=ThermometerTitle_GlobalServer
   properties {
    text=MEMORY USED
    font {
     bForceMonospaced=false
     bMaterialIsStencil=false
     fontMaterial=null
     fontObject=&NotoSans
     letterSpacing=0
     monospacedWidth=1
     size=16
     skewAmount=0
     typefaceFontName=NotoSans-Bold
     outlineSettings {
      bApplyOutlineToDropShadows=true
      bMiteredCorners=false
      bSeparateFillAlpha=false
      outlineMaterial=null
      outlineSize=1
      outlineColor=<rgba 0 0 0 0.75>
     }
    }
    margin=<ltrb 0 -2 2 -1>
   }
   slot {class=VerticalBoxSlot, properties {horizontalAlignment=HAlign_Center}}
  }
  ThermometerBar_GlobalServer {
   widgetClassPath=&/ThermometerBar
   bIsVariable=true
   parent=VerticalBox
   widget=ThermometerBar_GlobalServer
   properties {valueCurrent=999, valueMax=999, visibility=Collapsed}
   slot {class=VerticalBoxSlot}
  }
  CreativeSaveStatusDisplay {
   widgetClassPath=/FortniteCoreUI/Athena/HUD/LTM/Playground/CreativeSaveStatusDisplay
   bIsVariable=true
   parent=CanvasPanel_0
   widget=CreativeSaveStatusDisplay
   slot {
    class=CanvasPanelSlot
    properties {bAutoSize=true, layoutData {alignment=<1 0>, anchors {maximum=<1 0>, minimum=<1 0>}, offsets=<ltrb -378.7838 40 100 30>}}
   }
  }
  ThermometerArea_Preview {
   widgetClassPath=&ScaleBox
   bIsVariable=true
   parent=CanvasPanel_0
   widget=ThermometerArea_Preview
   properties {stretch=UserSpecified, userSpecifiedScale=0.8, visibility=Collapsed}
   slot {
    class=CanvasPanelSlot
    properties {bAutoSize=true, layoutData {alignment=<1 0>, anchors {maximum=<0.5 0.5>, minimum=<0.5 0.5>}, offsets=<ltrb 115 -35 100 30>}}
   }
  }
  PreviewBar_Memory {
   widgetClassPath=/FortniteCoreUI/HUD/LTM/Playground/Thermometer/PreviewBar_Memory
   bIsVariable=true
   parent=ThermometerArea_Preview
   widget=PreviewBar_Memory
   properties {visibility=Visible, renderTransform {angle=0, scale=<1.2 1.2>, shear=<0 0>, translation=<0 0>}}
   slot {class=ScaleBoxSlot}
  }
  TutorialHighlight {widgetClassPath=&TutorialHighlight}
 ]
}
