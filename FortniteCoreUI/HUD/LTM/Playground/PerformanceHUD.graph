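# Playground HUD element widget that hosts the legacy performance HUD and controls its cost preview visibility.
# On construct it becomes hit-test invisible and binds to the gameplay player controller's Creative UI metrics display change event.
# The visibility handler remaps the metrics display index into CostPreviewerIndex through a switch, then passes it to PerformanceHUD_Legacy.
# The legacy widget sits in a fill-stretched ScaleBox inside the root Overlay.
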WidgetBlueprint /FortniteCoreUI/HUD/LTM/Playground/PerformanceHUD
graphs [EventGraph {nodes [
   n1 = OnCostPreviewVisibilityChanged!() -> (OutputDelegate, then, UIMetricsDisplayIndex:int) {is=title K2Node_CustomEvent}
   > n2 = Switch on Int!(execute:Exec, Selection:int=@n1.UIMetricsDisplayIndex) -> (Default:Exec, 0:Exec, 1:Exec, 2:Exec)
   n3 = Set CostPreviewerIndex(execute=[@n2.0, @n2.Default], CostPreviewerIndex:int=1) -> (Output_Get:int=0)
   Set Costpreview Visibility!(
    execute=[@n5.then, @n3.then, @n4.then]
    self:Performance HUD Legacy Object Reference=Get PerformanceHUD_Legacy() -> (PerformanceHUD_Legacy:Performance HUD Legacy Object Reference)
    visibility:int=Get CostPreviewerIndex() -> (CostPreviewerIndex:int=0)
   ) -> (then) {is=title K2Node_CallFunction}
   n4 = Set CostPreviewerIndex(execute=@n2.1, CostPreviewerIndex=2) -> (Output_Get=0)
   n5 = Set CostPreviewerIndex(execute=@n2.2, CostPreviewerIndex=0) -> (Output_Get=0)
   Event Construct()
   > SetVisibility(InVisibility=HitTestInvisible)
   > n6 = Cast To FortPlayerControllerGameplay!(execute, Object=GetOwningPlayer()) -> (then, CastFailed, AsFort Player Controller Gameplay:Fort Player Controller Gameplay Object Reference) {is=title K2Node_DynamicCast}
   > Bind Event to On Creative UIMetrics Display Changed(self:Fort Player Controller Gameplay Object Reference=@n6, Delegate=Create Event()) {is=title K2Node_AddDelegate}
  ]}]
variables [CostPreviewerIndex {type=integer}]
widgets {
 info {inheritedWidgetCount=1, parentClass=&FortHUDElementWidget, rootWidgetClass=&Overlay, widgetCount=4}
 widgets [
  Overlay_140 {widgetClassPath=&Overlay, widget=Overlay_140}
  ScaleBox_460 {
   widgetClassPath=&ScaleBox
   parent=Overlay_140
   widget=ScaleBox_460
   properties {stretch=Fill}
   slot {class=OverlaySlot, properties {horizontalAlignment=HAlign_Fill, verticalAlignment=VAlign_Fill}}
  }
  PerformanceHUD_Legacy {
   widgetClassPath=&/PerformanceHUD_Legacy
   bIsVariable=true
   parent=ScaleBox_460
   widget=PerformanceHUD_Legacy
   slot {class=ScaleBoxSlot}
  }
  TutorialHighlight {widgetClassPath=&TutorialHighlight}
 ]
}
