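# Playground LTM HUD widget that shows a red budget-limit warning beside an icon on the memory thermometer.
# Three setter functions switch the text between target cell, target area and global memory messages.
# PlayWarningLoad plays the TipsLoad animation, which pops in the icon, stretches out the text bar, then scales up the text.
# PlayWarningOut plays that same animation in reverse, first stopping it if it is still playing.
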WidgetBlueprint /FortniteCoreUI/Athena/HUD/LTM/Playground/Thermometer/ThermometerWarningTips
animations [TipsLoad {
  binding_count=3
  length_seconds=0.75
  object=&self:TipsLoad
  sequence {bindings [
    ThresholdText {
     id=#1
     possessed_class=CommonTextBlock
     tracks [{
       type=MovieScene2DTransformTrack
       sections [{
         type=MovieScene2DTransformSection
         end_frame=15
         start_frame=0
         channels [
          Translation.X {type=MovieSceneScriptingFloatChannel, const [0,1]}
          Translation.Y {type=MovieSceneScriptingFloatChannel, const [0,1]}
          Angle {type=MovieSceneScriptingFloatChannel, const [0,1]}
          Scale.X {type=MovieSceneScriptingFloatChannel, keys [[0,0],[9,0],[11,1]]}
          Scale.Y {type=MovieSceneScriptingFloatChannel, keys [[0,0],[9,0],[11,1]]}
          Shear.X {type=MovieSceneScriptingFloatChannel, const [0,1]}
          Shear.Y {type=MovieSceneScriptingFloatChannel, const [0,1]}
         ]
        }]
      }]
    }
    HorizontalBox_0 {
     id=#2
     possessed_class=HorizontalBox
     tracks [{
       type=MovieScene2DTransformTrack
       sections [{
         type=MovieScene2DTransformSection
         end_frame=15
         start_frame=0
         channels [
          Translation.X {type=MovieSceneScriptingFloatChannel, const [0,1]}
          Translation.Y {type=MovieSceneScriptingFloatChannel, const [0,1]}
          Angle {type=MovieSceneScriptingFloatChannel, const [0,1]}
          Scale.X {type=MovieSceneScriptingFloatChannel, keys [[0,0],[5,0],[7,1],[9,0.9],[11,1]]}
          Scale.Y {type=MovieSceneScriptingFloatChannel, const [1,1]}
          Shear.X {type=MovieSceneScriptingFloatChannel, const [0,1]}
          Shear.Y {type=MovieSceneScriptingFloatChannel, const [0,1]}
         ]
        }]
      }]
    }
    SizeBox_104 {
     id=#3
     possessed_class=SizeBox
     tracks [{
       type=MovieScene2DTransformTrack
       sections [{
         type=MovieScene2DTransformSection
         end_frame=15
         start_frame=0
         channels [
          Translation.X {type=MovieSceneScriptingFloatChannel, const [0,1]}
          Translation.Y {type=MovieSceneScriptingFloatChannel, const [0,1]}
          Angle {type=MovieSceneScriptingFloatChannel, const [0,1]}
          Scale.X {type=MovieSceneScriptingFloatChannel, keys [[0,0],[5,1],[7,0.9],[9,1]]}
          Scale.Y {type=MovieSceneScriptingFloatChannel, keys [[0,0],[5,1],[7,0.9],[9,1]]}
          Shear.X {type=MovieSceneScriptingFloatChannel, const [0,1]}
          Shear.Y {type=MovieSceneScriptingFloatChannel, const [0,1]}
         ]
        }]
      }]
    }
   ]}
 }]
graphs [
 EventGraph {nodes [
   Event PreConstruct()
   Event Construct()
   Event Tick()
  ]}
 PlayWarningLoad {nodes [
   PlayWarningLoad() {is=title K2Node_FunctionEntry}
   > PlayAnimation(InAnimation=Get TipsLoad() -> (TipsLoad:Widget Animation Object Reference))
  ]}
 PlayWarningOut {nodes [
   PlayWarningOut() {is=title K2Node_FunctionEntry}
   > n1 = Branch(Condition=IsAnimationPlaying(InAnimation=Get TipsLoad()))
   > StopAnimation(InAnimation=Get TipsLoad())
   n2 = Get TipsLoad()
   > PlayAnimationTimeRange!(
    execute
    self:User Widget Object Reference
    InAnimation:Widget Animation Object Reference=@n2
    StartAtTime:float=0
    EndAtTime:float=GetEndTime(self=@n2)
    NumLoopsToPlay:int=1
    PlayMode:EUMGSequencePlayMode Enum=Reverse
    PlaybackSpeed:float=1
    bRestoreState:bool=false
   ) -> (then, ReturnValue:Widget Animation Handle Structure)
   PlayAnimation(execute=@n1.else, InAnimation=Get TipsLoad(), PlayMode=Reverse)
  ]}
 Set Cell text {nodes [
   Set Cell text() {is=title K2Node_FunctionEntry}
   > "SetText (Text)"(self=Get ThresholdText() -> (ThresholdText:Common Text Object Reference), InText="LOCTEXT(\"TARGET CELL HAS REACHED ITS BUDGET LIMIT \")")
  ]}
 Set GlobalText {nodes [
   Set GlobalText() {is=title K2Node_FunctionEntry}
   > "SetText (Text)"(self=Get ThresholdText(), InText="LOCTEXT(\"GLOBAL MEMORY HAS REACHED ITS BUDGET LIMIT \")")
  ]}
 Set Grouptext {nodes [
   Set Grouptext() {is=title K2Node_FunctionEntry}
   > "SetText (Text)"(self=Get ThresholdText(), InText="LOCTEXT(\"TARGET AREA HAS REACHED ITS BUDGET LIMIT \")")
  ]}
]
variables []
widgets {
 info {inheritedWidgetCount=0, parentClass=&UserWidget, rootWidgetClass=&SizeBox, widgetCount=8}
 widgets [
  SizeBox_1 {widgetClassPath=&SizeBox, widget=SizeBox_1, properties {bOverride_WidthOverride=true, widthOverride=290}}
  Overlay_0 {widgetClassPath=&Overlay, parent=SizeBox_1, widget=Overlay_0, slot {class=SizeBoxSlot}}
  SizeBox_0 {
   widgetClassPath=&SizeBox
   parent=Overlay_0
   widget=SizeBox_0
   properties {bOverride_WidthOverride=true, widthOverride=236}
   slot {class=OverlaySlot, properties {horizontalAlignment=HAlign_Fill, verticalAlignment=VAlign_Center, padding=<ltrb 55 0 0 0>}}
  }
  HorizontalBox_0 {
   widgetClassPath=&HorizontalBox
   parent=SizeBox_0
   widget=HorizontalBox_0
   properties {renderTransformPivot=<0 0>}
   slot {class=SizeBoxSlot}
  }
  CommonBorder_0 {
   widgetClassPath=&CommonBorder
   bIsVariable=true
   parent=HorizontalBox_0
   widget=CommonBorder_0
   properties {
    horizontalAlignment=HAlign_Center
    verticalAlignment=VAlign_Center
    background {
     bIsDynamicallyLoaded=false
     drawAs=Box
     imageType=NoImage
     mirroring=NoMirror
     resourceName=None
     resourceObject=null
     tiling=NoTile
     imageSize=<96 96>
     margin=<ltrb 0.25 0.25 0.25 0.25>
     outlineSettings {
      bUseBrushTransparency=false
      roundingType=HalfHeightRadius
      width=0
      color {colorUseRule=UseColor_Specified, specifiedColor=<rgba 0 0 0 0>}
      cornerRadii=<0 0 0 0>
     }
     tintColor {colorUseRule=UseColor_Foreground, specifiedColor=<rgba 1 1 1 1>}
     uVRegion {bIsValid=false, max=<0 0>, min=<0 0>}
    }
    brushColor=<rgba 0 0 0 0.8>
    padding=<ltrb 15 -2 7 -3>
   }
   slot {class=HorizontalBoxSlot, properties {verticalAlignment=VAlign_Center}}
  }
  ThresholdText {
   widgetClassPath=&CommonTextBlock
   bIsVariable=true
   parent=CommonBorder_0
   widget=ThresholdText
   properties {
    scrollStyle=&TextScrollStyle-Base_C
    text="TARGET AREA HAS REACHED ITS BUDGET LIMIT "
    colorAndOpacity {colorUseRule=UseColor_Specified, specifiedColor=<rgba 1 0 0.0012 1>}
    font {
     bForceMonospaced=false
     bMaterialIsStencil=false
     fontMaterial=null
     fontObject=&BurbankSmall
     letterSpacing=0
     monospacedWidth=1
     size=15
     skewAmount=0
     typefaceFontName=NotoSans-Regular
     outlineSettings {
      bApplyOutlineToDropShadows=false
      bMiteredCorners=false
      bSeparateFillAlpha=false
      outlineMaterial=null
      outlineSize=1
      outlineColor=<rgba 0.15 0 0 1>
     }
    }
   }
   slot {class=BorderSlot, properties {horizontalAlignment=HAlign_Center, verticalAlignment=VAlign_Center, padding=<ltrb 15 -2 7 -3>}}
  }
  SizeBox_104 {
   widgetClassPath=&SizeBox
   parent=Overlay_0
   widget=SizeBox_104
   slot {class=OverlaySlot, properties {verticalAlignment=VAlign_Center, padding=<ltrb -10 0 0 0>}}
  }
  ThermometerWarningIconWidget {
   widgetClassPath=&/ThermometerWarningIconWidget
   bIsVariable=true
   parent=SizeBox_104
   widget=ThermometerWarningIconWidget
   slot {class=SizeBoxSlot, properties {horizontalAlignment=HAlign_Center, verticalAlignment=VAlign_Center}}
  }
 ]
}
