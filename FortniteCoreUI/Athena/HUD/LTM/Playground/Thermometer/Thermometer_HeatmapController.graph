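# Creative HUD widget that controls the Thermometer performance heatmap and its toggle keybind, showing the keybind while the player flies.
# Construct spawns a Thermometer_HeatMap3D actor and binds keybind, toggle and display-mode events.
# RefreshHeatmap feeds player position and the map manager's mask into MPC_heatmap and the actor's materials, drawing to a render target.
# SetHeatmapUIScale pushes volume bounds to both; SetHeatmapVisibility switches the effect, clearing the target and stopping FX when off.
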WidgetBlueprint /FortniteCoreUI/Athena/HUD/LTM/Playground/Thermometer/Thermometer_HeatmapController
graphs [
 EventGraph {nodes [
   SetScalarParameterValue!(
    execute
    Collection:Material Parameter Collection Object Reference=/Game/Athena/HUD/LTM/Playground/Thermometer/MPC_heatmap
    ParameterName:Name=HeatmapGamma
    ParameterValue:float=1
   ) -> (then) {is=title K2Node_CallMaterialParameterCollectionFunction}
   > SetVisibility(InVisibility=HitTestInvisible)
   > SetHeatmapVisibility!(execute, self, EnableHeatmapVisibility:bool=true) -> (then)
   n1 = OnHeatmapToggled!() -> (OutputDelegate, then, bIsHeatmapToggled:bool=~) {is=title K2Node_CustomEvent}
   > n2 = Set IsHeatmapToggled(IsHeatmapToggled:bool=@n1.bIsHeatmapToggled) -> (Output_Get:bool)
   > Branch(Condition=@n2)
   > SetScalarParameterValue(Collection=&MPC_heatmap, ParameterName=MInnimapUvScale, ParameterValue=1)
   DisplayHeatmap() {is=title K2Node_CustomEvent}
   > Branch()
   > Do Once(Reset=@n3.then)
   > Call Playpostprocess!(
    execute:Exec
    self:Thermometer Heat Map 3D Object Reference=Get 3dHeatmapObject() -> (3dHeatmapObject:Thermometer Heat Map 3D Object Reference)
   ) -> (then:Exec) {is=title K2Node_CallDelegate}
   > SetScalarParameterValue(Collection=&MPC_heatmap, ParameterName=HeatMapSwitch, ParameterValue=1)
   > n3 = Delay(Duration=6)
   n4 = OnHeatmapDisplayModeChanged!() -> (OutputDelegate, then, UIMetricsDisplayIndex:int) {is=title K2Node_CustomEvent}
   > n5 = Switch on Int!(execute:Exec, Selection:int=@n4.UIMetricsDisplayIndex) -> (0:Exec, 1:Exec) {bHasDefaultPin=false}
   > SetScalarParameterValue(Collection=&MPC_heatmap, ParameterName=HeatmapGamma, ParameterValue=1)
   > SetHeatmapVisibility(EnableHeatmapVisibility=false)
   > n6 = SetVisibility(InVisibility=Collapsed)
   PrintString!(
    execute=[@n7.then, @n6.then]
    InString=Append(
     A="Current Heatmap Display Mode is: "
     B=Select(Option 0:String=OFF, Option 1:String=ON, Index:int=@n4.UIMetricsDisplayIndex) -> (ReturnValue:String)
    )
    bPrintToScreen
    bPrintToLog
    TextColor
    Duration
    Key
   ) -> (then)
   SetScalarParameterValue(execute=@n5.1, Collection=&MPC_heatmap, ParameterName=HeatmapGamma, ParameterValue=1)
   > SetHeatmapVisibility(EnableHeatmapVisibility=true)
   > n7 = SetVisibility(InVisibility=HitTestInvisible)
   Event Construct()
   > Branch(Condition=GetValueAsBool!(Input:"Scalable Float Structure (by ref)"=Get ShouldShowHeatmap() -> (ShouldShowHeatmap:Scalable Float Structure), Level:float=0) -> (ReturnValue:bool))
   > SetVisibility(InVisibility=Collapsed)
   > SpawnActor Thermometer Heat Map 3D!(
    execute:Exec
    Class:Actor Class Reference=/Game/Athena/HUD/LTM/Playground/Thermometer/Thermometer_HeatMap3D.Thermometer_HeatMap3D_C
    SpawnTransform:Transform=MakeTransform(Rotation="0, 0, 0")
    CollisionHandlingOverride:ESpawnActorCollisionHandlingMethod Enum=AlwaysSpawn
    TransformScaleMethod:ESpawnActorScaleMethod Enum=OverrideRootScale
    Owner:Actor Object Reference
    Instigator:Pawn Object Reference
   ) -> (then:Exec, ReturnValue:Thermometer Heat Map 3D Object Reference) {is=title K2Node_SpawnActorFromClass}
   > n8 = GetActorOfClass!(execute, ActorClass:Actor Class Reference=&Thermometer_HeatMap3D_C) -> (then, ReturnValue:Thermometer Heat Map 3D Object Reference)
   > Set 3dHeatmapObject(3dHeatmapObject:Thermometer Heat Map 3D Object Reference=@n8) -> (Output_Get:Thermometer Heat Map 3D Object Reference)
   > SetHeatmapBoundAction()
   > Bind Event to On Keybinds Changed(
    self:Fort Global UIContext Object Reference=Get FortGlobalUIContext() -> (ReturnValue:Fort Global UIContext Object Reference) {is=title K2Node_GetSubsystem}
    Delegate=Create Event()
   ) {is=title K2Node_AddDelegate}
   > Bind Event to On Creative Heatmap Toggled(
    self:Creative HUDLContext Object Reference=Get CreativeHUDLContext!() -> (ReturnValue:Creative HUDLContext Object Reference) {is=title K2Node_GetBlueprintContextEx}
    Delegate=Create Event()
   ) {is=title K2Node_AddDelegate}
   n9 = Event RefreshHeatmap!() -> (OutputDelegate, then, PlayerPosition:"Vector (by ref)"="0, 0, 0") {is=title K2Node_Event}
   > Get!(execute:Exec) -> (then:Exec, else:Exec, 3dHeatmapObject:Thermometer Heat Map 3D Object Reference)
   > SetVisibility(InVisibility=HitTestInvisible)
   > n10 = Cast To FortPlayerPawnAthena(Object=GetPlayerPawn!(PlayerIndex:int=0) -> (ReturnValue:Pawn Object Reference))
   > SetRenderCustomDepth!(execute, self:Primitive Component Object Reference=Get Mesh(self=@n10), bValue:bool=true) -> (then)
   > n11 = Sequence!(execute) -> (then_0, then_1, then_2:Exec, then_3:Exec, then_4:Exec)
   n12 = BreakVector(InVec=@n9.PlayerPosition)
   > SetScalarParameterValue(Collection=&MPC_heatmap, ParameterName=PlayerX, ParameterValue=@n12)
   > SetScalarParameterValue(Collection=&MPC_heatmap, ParameterName=PlayerY, ParameterValue=@n12.Y)
   n13 = GetHeatmapMask!(self:Fort In Game Map Manager Object Reference=GetUIMapManager!() -> (ReturnValue:Fort In Game Map Manager Object Reference)) -> (ReturnValue:Texture 2D Object Reference)
   > Set Tex_heatmap_code!(execute, Tex_heatmap_code:Texture 2D Object Reference=@n13, self:Thermometer Heat Map 3D Object Reference=Get 3dHeatmapObject()) -> (then, Output_Get:Texture 2D Object Reference)
   > SetTextureParameterValue(
    self=Get HeatmapUI32!(self:Thermometer Heat Map 3D Object Reference=Get 3dHeatmapObject()) -> (HeatmapUI32:Material Instance Dynamic Object Reference)
    ParameterName=Heatmap_code
    Value=@n13
   )
   n14 = Get HeatMapUI!(self:Thermometer Heat Map 3D Object Reference=Get 3dHeatmapObject()) -> (HeatMapUI:Material Instance Dynamic Object Reference)
   > SetTextureParameterValue(self=@n14, ParameterName=Heatmap_code, Value=@n13)
   > DrawMaterialToRenderTarget!(
    execute
    TextureRenderTarget:Texture Render Target 2D Object Reference=/Game/Athena/HUD/LTM/Playground/Thermometer/RT_Heatmapcode
    Material:Material Interface Object Reference=@n14
   ) -> (then)
   n15 = GetUIMapManager()
   n16 = GetHeatmapMaskScale!(self:Fort In Game Map Manager Object Reference=@n15) -> (ReturnValue:float)
   SetScalarParameterValue(execute=@n11.then_1, Collection=&MPC_heatmap, ParameterName=HeatmapMaskScale, ParameterValue=@n16)
   n17 = GetHeatmapMask(self=@n15)
   > Is Valid(InputObject=@n17)
   n18 = GetSizeX!(self:Texture 2D Object Reference=@n17) -> (ReturnValue:int=0)
   > SetScalarParameterValue(
    Collection=&MPC_heatmap
    ParameterName=Gridsize
    ParameterValue=float * float(A=@n16, B="To Float (Integer)"(InInt=@n18))
   )
   Branch(execute=@n11.then_2, Condition=Get IsHeatmapToggled() -> (IsHeatmapToggled:bool))
   > DisplayHeatmap() {is=title K2Node_CallFunction}
   n19 = Branch(execute=@n11.then_3, Condition=Get IsInPreviewMode() -> (IsInPreviewMode:bool))
   n20 = BreakVector(InVec=@n9.PlayerPosition)
   SetScalarParameterValue(execute=@n19.else, Collection=&MPC_heatmap, ParameterName=AimTargetX, ParameterValue=@n20)
   > SetScalarParameterValue(Collection=&MPC_heatmap, ParameterName=AimTargetY, ParameterValue=@n20.Y)
   n21 = Cast To FortPlayerPawnAthena(execute=@n11.then_4, Object=GetOwningPlayerPawn!(self:User Widget Object Reference) -> (ReturnValue:Pawn Object Reference))
   > n22 = Switch on EMovementMode!(
    execute
    Selection:EMovementMode Enum=Get MovementMode!(self:Character Movement Component Object Reference=Get CharacterMovement!(self:Character Object Reference=@n21) -> (CharacterMovement:Character Movement Component Object Reference)) -> (MovementMode:EMovementMode Enum=MOVE_None)
   ) -> (MOVE_None:Exec, MOVE_Walking:Exec, MOVE_NavWalking:Exec, MOVE_Falling:Exec, MOVE_Swimming:Exec, MOVE_Flying:Exec, MOVE_Custom:Exec) {is=title K2Node_SwitchEnum}
   n23 = Get HeatmapToggle_Keybind() -> (HeatmapToggle_Keybind:Keybind Widget Object Reference)
   SetVisibility(
    execute=[@n22.MOVE_Custom, @n22.MOVE_Falling, @n22.MOVE_NavWalking, @n22.MOVE_None, @n22.MOVE_Swimming, @n22.MOVE_Walking]
    self=@n23
    InVisibility=Hidden
   )
   n24 = Get HeatmapToggle_FeedbackText() -> (HeatmapToggle_FeedbackText:Rich Text Block Object Reference)
   > SetVisibility(self=@n24, InVisibility=Hidden)
   n25 = Get HeatmapToggle_Icon() -> (HeatmapToggle_Icon:Image Object Reference)
   > SetVisibility(self=@n25, InVisibility=Hidden)
   SetVisibility(execute=@n22.MOVE_Flying, self=@n23, InVisibility=Visible)
   > SetVisibility(self=@n24, InVisibility=Hidden)
   > SetVisibility(self=@n25, InVisibility=Visible)
   Bind Event to On Creative UIMetrics Display Changed(
    self:Fort Player Controller Gameplay Object Reference=Cast To FortPlayerControllerGameplay!(execute, Object=GetOwningPlayer()) -> (then, CastFailed, AsFort Player Controller Gameplay:Fort Player Controller Gameplay Object Reference) {is=title K2Node_DynamicCast}
    Delegate=Create Event()
   ) {is=title K2Node_AddDelegate}
   Get IsHeatmapEnabledOnOption() -> (IsHeatmapEnabledOnOption:bool)
  ]}
 SetHeatmapBoundAction {nodes [
   SetHeatmapBoundAction() {is=title K2Node_FunctionEntry}
   n1 = Get FortGlobalUIContext()
   > Is Valid(InputObject=@n1)
   n2 = Get HeatmapToggle_Keybind()
   > SetBoundAction!(
    execute
    self:Fort Keybind Widget Object Reference=@n2
    NewBoundAction:Name=GetCreativeToggleHeatmapName!(self:Fort Global UIContext Object Reference=@n1) -> (ReturnValue:Name=None)
   ) -> (then)
   > UpdateKeybindWidget(self:Fort Keybind Widget Object Reference=@n2)
  ]}
 SetHeatmapUIScale {nodes [
   SetHeatmapUIScale() {is=title K2Node_FunctionEntry}
   > n1 = Sequence!(execute) -> (then_0, then_1)
   n2 = BreakVector(InVec=GetCurrentVolumeSize!(self:Creative Performance HUDHeatmap Control Widget Object Reference) -> (ReturnValue:Vector))
   > SetScalarParameterValue(Collection=&MPC_heatmap, ParameterName=VolumeSizeX, ParameterValue=@n2)
   n3 = Get 3dHeatmapObject()
   > n4 = Set volume size!(execute, volume size:double=@n2, self:Thermometer Heat Map 3D Object Reference=@n3) -> (then, Output_Get:double)
   > n5 = Sequence()
   > DisplayHeatmap()
   > Cast To FortGameStateAthena!(execute, Object=GetGameState!() -> (ReturnValue:Game State Base Object Reference)) -> (then, CastFailed, AsFort Game State Athena:Fort Game State Athena Object Reference) {is=title K2Node_DynamicCast}
   > SetScalarParameterValue(Collection=&MPC_heatmap, ParameterName=MInnimapUvScale, ParameterValue=1)
   > SetScalarParameterValue(Collection=&MPC_heatmap, ParameterName=ModeSWitch, ParameterValue=0)
   n6 = GetHeatmapMaskScale(self=GetUIMapManager())
   > PrintString(
    InString="To String (Float)"(InDouble=float / float(A=1, B=@n6) {is=title K2Node_CallFunction})
    TextColor=(R=1,G=0,B=0.0245,A=1)
    Duration=5
   )
   n7 = Compare Float!(Exec:Exec=@n5.then_1, Input:double=64000, CompareWith:double=@n4) -> (">":Exec, "==":Exec, "<":Exec) {is=title K2Node_MacroInstance}
   SetScalarParameterValue(execute=@n7."<", Collection=&MPC_heatmap, ParameterName=Island Size, ParameterValue=0)
   SetScalarParameterValue(execute=@n7.">", Collection=&MPC_heatmap, ParameterName=Island Size, ParameterValue=1)
   n8 = Break Box!(Box:"Box Structure (by ref)"=GetCurrentVolumeBox!(self:Creative Performance HUDHeatmap Control Widget Object Reference) -> (ReturnValue:Box Structure)) -> (Min:Vector=~, Max:Vector=~, IsValid:bool=~) {is=title K2Node_BreakStruct}
   n9 = BreakVector(InVec=@n8)
   n10 = Set minsizeX!(execute=@n1.then_1, minsizeX:double=@n9, self:Thermometer Heat Map 3D Object Reference=@n3) -> (then, Output_Get:double)
   > SetScalarParameterValue(Collection=&MPC_heatmap, ParameterName=volumeSizeMInX, ParameterValue=@n10)
   > n11 = Set minsizeY!(execute, minsizeY:double=@n9.Y, self:Thermometer Heat Map 3D Object Reference=@n3) -> (then, Output_Get:double)
   > SetScalarParameterValue(Collection=&MPC_heatmap, ParameterName=volumeSizeMInY, ParameterValue=@n11)
   n12 = BreakVector(InVec=@n8.Max)
   > n13 = Set MaxsizeX!(execute, MaxsizeX:double=@n12, self:Thermometer Heat Map 3D Object Reference=@n3) -> (then, Output_Get:double)
   > SetScalarParameterValue(Collection=&MPC_heatmap, ParameterName=volumeSizeMaxX, ParameterValue=@n13)
   > n14 = Set MaxsizeY!(execute, MaxsizeY:double=@n12.Y, self:Thermometer Heat Map 3D Object Reference=@n3) -> (then, Output_Get:double)
   > SetScalarParameterValue(Collection=&MPC_heatmap, ParameterName=volumeSizeMaxY, ParameterValue=@n14)
  ]}
 SetHeatmapVisibility {nodes [
   n1 = SetHeatmapVisibility!() -> (then, EnableHeatmapVisibility:bool=~) {is=title K2Node_FunctionEntry}
   > n2 = Set IsHeatmapEnabledOnOption(IsHeatmapEnabledOnOption:bool=@n1) -> (Output_Get:bool)
   > n3 = Branch(Condition=@n2)
   > SetScalarParameterValue(Collection=&MPC_heatmap, ParameterName=HeatMapSwitch, ParameterValue=1)
   n4 = Get HeatmapToggle_Keybind()
   > SetVisibility(self=@n4, InVisibility=HitTestInvisible)
   n5 = Get HeatmapToggle_Icon()
   > SetVisibility(self=@n5, InVisibility=Visible)
   > Return Node()
   SetScalarParameterValue(execute=@n3.else, Collection=&MPC_heatmap, ParameterName=HeatMapSwitch, ParameterValue=0)
   > SetVisibility(self=@n4, InVisibility=Hidden)
   > SetVisibility(self=@n5, InVisibility=Hidden)
   > ClearRenderTarget2D!(
    execute
    TextureRenderTarget:Texture Render Target 2D Object Reference=&RT_Heatmapcode
    ClearColor:Linear Color Structure=(R=0,G=0,B=0,A=1)
   ) -> (then)
   > n6 = Is Valid(InputObject=Get 3dHeatmapObject())
   > n7 = StopFX(self:Thermometer Heat Map 3D Object Reference=Get 3dHeatmapObject())
   Return Node(execute=[@n6.Is Not Valid, @n7.then])
  ]}
]
variables [
 3dHeatmapObject {type=Thermometer_HeatMap3D_C, default=null}
 IsHeatmapEnabledOnOption {type=boolean}
 IsHeatmapToggled {type=boolean}
 IsInPreviewMode {type=boolean, default=true}
 ShouldShowHeatmap {
  type=ScalableFloat
  default {
   registryType=None
   value=1
   curve {curveTable=/Game/Athena/Playlists/Creative/AthenaCompositeGD_Creative, rowName=Creative.Thermometer.3DHeatmap.Enabled}
  }
 }
]
widgets {
 info {
  inheritedWidgetCount=1
  parentClass=/Script/FortniteUI.CreativePerformanceHUDHeatmapControlWidget
  rootWidgetClass=&HorizontalBox
  widgetCount=6
 }
 widgets [
  HorizontalBox_0 {widgetClassPath=&HorizontalBox, widget=HorizontalBox_0}
  InventoryRow {
   widgetClassPath=&HorizontalBox
   parent=HorizontalBox_0
   widget=InventoryRow
   slot {class=HorizontalBoxSlot, properties {horizontalAlignment=HAlign_Right}}
  }
  HeatmapToggle_FeedbackText {
   widgetClassPath=/Script/UMG.RichTextBlock
   bIsVariable=true
   parent=InventoryRow
   widget=HeatmapToggle_FeedbackText
   properties {
    bOverrideDefaultStyle=true
    text=Fly to Use
    textStyleSet=/Game/Athena/HUD/TextStyles/PlayersLeftStyles
    defaultTextStyleOverride {
     overflowPolicy=Clip
     transformPolicy=None
     colorAndOpacity {colorUseRule=UseColor_Specified, specifiedColor=<rgba 1 1 1 1>}
     font {
      bForceMonospaced=false
      bMaterialIsStencil=false
      fontMaterial=null
      fontObject=&BurbankSmall
      letterSpacing=0
      monospacedWidth=1
      size=18
      skewAmount=0
      typefaceFontName=Bold
      outlineSettings {
       bApplyOutlineToDropShadows=true
       bMiteredCorners=false
       bSeparateFillAlpha=true
       outlineMaterial=&M-FontShadow
       outlineSize=3
       outlineColor=<rgba 0 0 0 0.5>
      }
     }
     highlightColor {colorUseRule=UseColor_Specified, specifiedColor=<rgba 0 0 0 1>}
     highlightShape {
      bIsDynamicallyLoaded=false
      drawAs=Image
      imageType=NoImage
      mirroring=NoMirror
      resourceName=None
      resourceObject=null
      tiling=NoTile
      imageSize=<32 32>
      margin=<ltrb 0 0 0 0>
      outlineSettings {
       bUseBrushTransparency=false
       roundingType=HalfHeightRadius
       width=0
       color {colorUseRule=UseColor_Specified, specifiedColor=<rgba 0 0 0 0>}
       cornerRadii=<0 0 0 0>
      }
      tintColor {colorUseRule=UseColor_Specified, specifiedColor=<rgba 1 1 1 1>}
      uVRegion {bIsValid=false, max=<0 0>, min=<0 0>}
     }
     selectedBackgroundColor {colorUseRule=UseColor_Foreground, specifiedColor=<rgba 1 0 1 0>}
     shadowColorAndOpacity=<rgba 0 0 0 1>
     shadowOffset=<0 0>
     strikeBrush {
      bIsDynamicallyLoaded=false
      drawAs=Image
      imageType=NoImage
      mirroring=NoMirror
      resourceName=None
      resourceObject=null
      tiling=NoTile
      imageSize=<32 32>
      margin=<ltrb 0 0 0 0>
      outlineSettings {
       bUseBrushTransparency=false
       roundingType=HalfHeightRadius
       width=0
       color {colorUseRule=UseColor_Specified, specifiedColor=<rgba 0 0 0 0>}
       cornerRadii=<0 0 0 0>
      }
      tintColor {colorUseRule=UseColor_Specified, specifiedColor=<rgba 1 1 1 1>}
      uVRegion {bIsValid=false, max=<0 0>, min=<0 0>}
     }
     underlineBrush {
      bIsDynamicallyLoaded=false
      drawAs=Image
      imageType=NoImage
      mirroring=NoMirror
      resourceName=None
      resourceObject=null
      tiling=NoTile
      imageSize=<32 32>
      margin=<ltrb 0 0 0 0>
      outlineSettings {
       bUseBrushTransparency=false
       roundingType=HalfHeightRadius
       width=0
       color {colorUseRule=UseColor_Specified, specifiedColor=<rgba 0 0 0 0>}
       cornerRadii=<0 0 0 0>
      }
      tintColor {colorUseRule=UseColor_Specified, specifiedColor=<rgba 1 1 1 1>}
      uVRegion {bIsValid=false, max=<0 0>, min=<0 0>}
     }
    }
    renderTransform {angle=0, scale=<1 1>, shear=<0 0>, translation=<36 0>}
   }
   slot {class=HorizontalBoxSlot, properties {horizontalAlignment=HAlign_Right, verticalAlignment=VAlign_Center}}
  }
  HeatmapToggle_Keybind {
   widgetClassPath=/FortUILibrary/Input/KeybindWidget
   bIsVariable=true
   parent=InventoryRow
   widget=HeatmapToggle_Keybind
   properties {boundAction=ToggleHeatmapRefresh, boundKeyFallback=T}
   slot {class=HorizontalBoxSlot, properties {size {sizeRule=Fill, value=1}}}
  }
  HeatmapToggle_Icon {
   widgetClassPath=&Image
   bIsVariable=true
   parent=InventoryRow
   widget=HeatmapToggle_Icon
   properties {brush {
     bIsDynamicallyLoaded=false
     drawAs=Image
     imageType=NoImage
     mirroring=NoMirror
     resourceName=None
     resourceObject=/Game/Athena/HUD/LTM/Playground/Thermometer/M_HeatmapIcon
     tiling=NoTile
     imageSize=<35 27>
     margin=<ltrb 0 0 0 0>
     outlineSettings {
      bUseBrushTransparency=false
      roundingType=HalfHeightRadius
      width=0
      color {colorUseRule=UseColor_Specified, specifiedColor=<rgba 0 0 0 0>}
      cornerRadii=<0 0 0 0>
     }
     tintColor {colorUseRule=UseColor_Specified, specifiedColor=<rgba 1 1 1 1>}
     uVRegion {bIsValid=false, max=<0 0>, min=<0 0>}
    }}
   slot {class=HorizontalBoxSlot, properties {verticalAlignment=VAlign_Center, padding=<ltrb 4 0 0 0>}}
  }
  TutorialHighlight {widgetClassPath=&TutorialHighlight}
 ]
}
