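# Creative HUD warning widget in the ThermometerV2 set, showing an icon and red "HEAT MAP IS OVER BUDGET!" text on a dark strip.
# When the native parent reports a changed over-budget state, it stores the value and plays Pop forward or in reverse from the current time.
# Pop staggers scaling of icon, strip and text.
# On Pop At Start sets visibility from the stored state; On Pop At End is empty.
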WidgetBlueprint /FortniteCoreUI/Athena/HUD/LTM/Playground/ThermometerV2/WBP_HeatmapOverBudget_Warning
animations [Pop {
  binding_count=3
  length_seconds=0.75
  object=&self:Pop
  sequence {bindings [
    ThresholdText {
     id=#1
     possessed_class=CommonTextBlock
     tracks [{
       type=MovieScene2DTransformTrack
       sections [{
         type=MovieScene2DTransformSection
         end_frame=15
         start_frame=0
         channels [
          Translation.X {type=MovieSceneScriptingFloatChannel, const [0,1]}
          Translation.Y {type=MovieSceneScriptingFloatChannel, const [0,1]}
          Angle {type=MovieSceneScriptingFloatChannel, const [0,1]}
          Scale.X {type=MovieSceneScriptingFloatChannel, keys [[0,0],[9,0],[11,1]]}
          Scale.Y {type=MovieSceneScriptingFloatChannel, keys [[0,0],[9,0],[11,1]]}
          Shear.X {type=MovieSceneScriptingFloatChannel, const [0,1]}
          Shear.Y {type=MovieSceneScriptingFloatChannel, const [0,1]}
         ]
        }]
      }]
    }
    HorizontalBox_0 {
     id=#2
     possessed_class=HorizontalBox
     tracks [
      {
       type=MovieScene2DTransformTrack
       sections [{
         type=MovieScene2DTransformSection
         end_frame=15
         start_frame=0
         channels [
          Translation.X {type=MovieSceneScriptingFloatChannel, const [0,1]}
          Translation.Y {type=MovieSceneScriptingFloatChannel, const [0,1]}
          Angle {type=MovieSceneScriptingFloatChannel, const [0,1]}
          Scale.X {type=MovieSceneScriptingFloatChannel, keys [[0,0],[5,0],[7,1],[9,0.9],[11,1]]}
          Scale.Y {type=MovieSceneScriptingFloatChannel, const [1,1]}
          Shear.X {type=MovieSceneScriptingFloatChannel, const [0,1]}
          Shear.Y {type=MovieSceneScriptingFloatChannel, const [0,1]}
         ]
        }]
      }
      {
       type=MovieSceneEventTrack
       sections [{
         type=MovieSceneEventTriggerSection
         channels [None {
           type=MovieSceneScriptingEventChannel
           keys [[0, "<Struct 'MovieSceneEvent' (0x46742983020) {}>"], [15, "<Struct 'MovieSceneEvent' (0x46742982f80) {}>"]]
          }]
        }]
      }
     ]
    }
    SizeBox_104 {
     id=#3
     possessed_class=SizeBox
     tracks [{
       type=MovieScene2DTransformTrack
       sections [{
         type=MovieScene2DTransformSection
         end_frame=15
         start_frame=0
         channels [
          Translation.X {type=MovieSceneScriptingFloatChannel, const [0,1]}
          Translation.Y {type=MovieSceneScriptingFloatChannel, const [0,1]}
          Angle {type=MovieSceneScriptingFloatChannel, const [0,1]}
          Scale.X {type=MovieSceneScriptingFloatChannel, keys [[0,0],[5,1],[7,0.9],[9,1]]}
          Scale.Y {type=MovieSceneScriptingFloatChannel, keys [[0,0],[5,1],[7,0.9],[9,1]]}
          Shear.X {type=MovieSceneScriptingFloatChannel, const [0,1]}
          Shear.Y {type=MovieSceneScriptingFloatChannel, const [0,1]}
         ]
        }]
      }]
    }
   ]}
 }]
graphs [EventGraph {nodes [
   On Pop At Start() {is=title K2Node_CustomEvent}
   > SetVisibility(InVisibility=Select(Option 0=Collapsed, Option 1=SelfHitTestInvisible, Index=Get HeatmapOverBudget() -> (HeatmapOverBudget:bool)))
   n1 = Event On Heatmap Over Budget Changed!() -> (OutputDelegate, then, bHeatmapOverBudget:bool) {is=title K2Node_Event}
   > Branch(Condition="Not Equal (Boolean)"!(A:bool=@n1.bHeatmapOverBudget, B:bool=Get HeatmapOverBudget()) -> (ReturnValue:bool) {is=title K2Node_CallFunction})
   > Set HeatmapOverBudget(HeatmapOverBudget:bool=@n1.bHeatmapOverBudget) -> (Output_Get:bool)
   > PlayAnimation(
    InAnimation=Get Pop() -> (Pop:Widget Animation Object Reference)
    StartAtTime=GetAnimationCurrentTime!(self:User Widget Object Reference, InAnimation:Widget Animation Object Reference=Get Pop()) -> (ReturnValue:float)
    PlayMode=Select(Option 0:EUMGSequencePlayMode Enum=Reverse, Option 1:EUMGSequencePlayMode Enum, Index=@n1.bHeatmapOverBudget) -> (ReturnValue:EUMGSequencePlayMode Enum)
   )
   On Pop At End() {is=title K2Node_CustomEvent}
  ]}]
variables [HeatmapOverBudget {type=boolean}]
widgets {
 info {
  inheritedWidgetCount=0
  parentClass=/Script/FortniteUI.FortCreativeHeatmapOverBudgetWidget
  rootWidgetClass=&SizeBox
  widgetCount=8
 }
 widgets [
  SizeBox_1 {widgetClassPath=&SizeBox, widget=SizeBox_1, properties {bOverride_WidthOverride=true, widthOverride=290}}
  Overlay_0 {widgetClassPath=&Overlay, parent=SizeBox_1, widget=Overlay_0, slot {class=SizeBoxSlot}}
  SizeBox_0 {
   widgetClassPath=&SizeBox
   parent=Overlay_0
   widget=SizeBox_0
   properties {bOverride_WidthOverride=true, widthOverride=236}
   slot {class=OverlaySlot, properties {horizontalAlignment=HAlign_Fill, verticalAlignment=VAlign_Center, padding=<ltrb 55 0 0 0>}}
  }
  HorizontalBox_0 {
   widgetClassPath=&HorizontalBox
   parent=SizeBox_0
   widget=HorizontalBox_0
   properties {renderTransformPivot=<0 0>}
   slot {class=SizeBoxSlot}
  }
  CommonBorder_0 {
   widgetClassPath=&CommonBorder
   bIsVariable=true
   parent=HorizontalBox_0
   widget=CommonBorder_0
   properties {
    horizontalAlignment=HAlign_Center
    verticalAlignment=VAlign_Center
    background {
     bIsDynamicallyLoaded=false
     drawAs=Box
     imageType=NoImage
     mirroring=NoMirror
     resourceName=None
     resourceObject=null
     tiling=NoTile
     imageSize=<96 96>
     margin=<ltrb 0.25 0.25 0.25 0.25>
     outlineSettings {
      bUseBrushTransparency=false
      roundingType=HalfHeightRadius
      width=0
      color {colorUseRule=UseColor_Specified, specifiedColor=<rgba 0 0 0 0>}
      cornerRadii=<0 0 0 0>
     }
     tintColor {colorUseRule=UseColor_Foreground, specifiedColor=<rgba 1 1 1 1>}
     uVRegion {bIsValid=false, max=<0 0>, min=<0 0>}
    }
    brushColor=<rgba 0 0 0 0.8>
    padding=<ltrb 15 -2 7 -3>
   }
   slot {class=HorizontalBoxSlot, properties {verticalAlignment=VAlign_Center}}
  }
  ThresholdText {
   widgetClassPath=&CommonTextBlock
   bIsVariable=true
   parent=CommonBorder_0
   widget=ThresholdText
   properties {
    scrollStyle=&TextScrollStyle-Base_C
    text="HEAT MAP IS OVER BUDGET!"
    colorAndOpacity {colorUseRule=UseColor_Specified, specifiedColor=<rgba 1 0 0.0012 1>}
    font {
     bForceMonospaced=false
     bMaterialIsStencil=false
     fontMaterial=null
     fontObject=&BurbankSmall
     letterSpacing=0
     monospacedWidth=1
     size=15
     skewAmount=0
     typefaceFontName=NotoSans-Regular
     outlineSettings {
      bApplyOutlineToDropShadows=false
      bMiteredCorners=false
      bSeparateFillAlpha=false
      outlineMaterial=null
      outlineSize=1
      outlineColor=<rgba 0.15 0 0 1>
     }
    }
   }
   slot {class=BorderSlot, properties {horizontalAlignment=HAlign_Center, verticalAlignment=VAlign_Center, padding=<ltrb 15 -2 7 -3>}}
  }
  SizeBox_104 {
   widgetClassPath=&SizeBox
   parent=Overlay_0
   widget=SizeBox_104
   slot {class=OverlaySlot, properties {verticalAlignment=VAlign_Center, padding=<ltrb -10 0 0 0>}}
  }
  ThermometerWarningIconWidget {
   widgetClassPath=/FortniteCoreUI/Athena/HUD/LTM/Playground/Thermometer/ThermometerWarningIconWidget
   bIsVariable=true
   parent=SizeBox_104
   widget=ThermometerWarningIconWidget
   slot {class=SizeBoxSlot, properties {horizontalAlignment=HAlign_Center, verticalAlignment=VAlign_Center}}
  }
 ]
}
