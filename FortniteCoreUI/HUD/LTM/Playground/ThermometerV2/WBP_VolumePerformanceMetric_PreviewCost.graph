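# Creative thermometer HUD widget that previews the memory cost of the selected actor against the volume's budget.
# Metric and cost-preview events run an over-limit check, and a macro sets red or white text and a warning icon for over-limit or margin states.
# Cost updates also reposition PreviewBar_Memory by sign and show the absolute difference between current and projected memory, each normalized to a curve-table memory limit and rounded.
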WidgetBlueprint /FortniteCoreUI/HUD/LTM/Playground/ThermometerV2/WBP_VolumePerformanceMetric_PreviewCost
graphs [
 EventGraph {nodes [
   n1 = Event On Volume Performance Metrics Updated!() -> (OutputDelegate, then, VolumePerformanceMetrics:"Volume Performance Metrics Structure (by ref)") {is=title K2Node_Event}
   > n2 = CheckIfVolumeWillBeOverMemory!(execute, InMetrics:"Volume Performance Metrics Structure (by ref)"=@n1.VolumePerformanceMetrics) -> (then, bOutIsOverMemoryLimit:bool, bOutIsOverMemoryMargin:bool)
   > Update Preview Bar Visuals!(execute:Exec, IsOverMemoryLimit:bool=@n2, IsOverMemoryMargin:bool=@n2.bOutIsOverMemoryMargin) -> (then:Exec) {is=title K2Node_MacroInstance}
   n3 = Event On Cost Preview Of Selected Actor Updated!() -> (OutputDelegate, then, CostPreviewOfSelectedActor:int=0, HighestCellCostPreviewOfSelectedActor:int=0) {is=title K2Node_Event}
   > Set Cost Preview Of Selected Actor(Cost Preview Of Selected Actor:int=@n3.CostPreviewOfSelectedActor) -> (Output_Get:int=0)
   > n4 = CheckIfVolumeWillBeOverMemory(InMetrics=Get LastVolumePerformanceMetrics() -> (LastVolumePerformanceMetrics:Volume Performance Metrics Structure))
   > Update Preview Bar Visuals(IsOverMemoryLimit=@n4, IsOverMemoryMargin=@n4.bOutIsOverMemoryMargin)
   > n5 = Branch(Condition="integer < integer"(A=Get Cost Preview Of Selected Actor() -> (Cost Preview Of Selected Actor:int=0)))
   > n6 = SetRenderTransform!(
    execute
    self:Widget Object Reference=Get PreviewBar_Memory() -> (PreviewBar_Memory:Preview Bar Memory Object Reference)
    InTransform_Translation:Vector 2D Structure=(X=5,Y=33)
    InTransform_Scale:Vector 2D Structure=(X=1.3,Y=1.3)
    InTransform_Shear:Vector 2D Structure=(X=0,Y=0)
    InTransform_Angle:float=0
   ) -> (then)
   n7 = Break Volume Performance Metrics!(VolumePerformanceMetrics:"Volume Performance Metrics Structure (by ref)"=Get LastVolumePerformanceMetrics()) -> (
    PerformanceValue:int
    PerformanceMaxValue:int
    PerformanceValueMargin:int
    PerformanceValueCellIndex:Int Point Structure
    PerformanceLowendThreshold:int
    PreviewDeltaValue:int
    Category:EFortBudgetCategory Enum
    PlayerID:int
    MostExpensivePerformanceValue:int
    MostExpensivePreviewDeltaValue:int
    GlobalMemoryValue:int
    MostExpensivePerformanceValueCellIndex:Int Point Structure
   ) {is=title K2Node_BreakStruct}
   n8 = "To Float (Integer)"(InInt=@n7.PerformanceMaxValue)
   n9 = "To Float (Integer)"(InInt=@n7)
   n10 = "Clamp (Float)"(Value=@n9, Min=0, Max=@n9)
   n11 = Round(A=float * float(A=float / float(A=@n10, B=@n8) {is=title K2Node_CallFunction}, B=GetMemoryLimit!() -> (ThermometerBarVaxValue:double))) {is=title K2Node_CallFunction}
   n12 = Break Volume Performance Metrics(VolumePerformanceMetrics=Get LastVolumePerformanceMetrics())
   n13 = "To Float (Integer)"(InInt=int + int(A=@n3.CostPreviewOfSelectedActor, B=@n12) -> (ReturnValue=0) {is=title K2Node_CommutativeAssociativeBinaryOperator})
   n14 = "Clamp (Float)"(Value=@n13, Min=0, Max=@n13)
   n15 = "To Float (Integer)"(InInt=@n12.PerformanceMaxValue)
   n16 = Round(A=float * float(A=float / float(A=@n14, B=@n15), B=GetMemoryLimit()))
   Set previewText!(
    execute=[@n6.then, @n17.then]
    self:Preview Bar Memory Object Reference=Get PreviewBar_Memory()
    PreviewTextNUM:double=0
    PreviewTextNUM int:int="Absolute (Integer)"!(A:int=int - int(A=@n16, B=@n11) {is=title K2Node_CallFunction}) -> (ReturnValue:int=0)
    Plus or minus:bool="integer < integer"(A=Get Cost Preview Of Selected Actor() -> (Cost Preview Of Selected Actor=0))
   ) -> (then) {is=title K2Node_CallFunction}
   n17 = SetRenderTransform(
    execute=@n5.else
    self=Get PreviewBar_Memory()
    InTransform_Translation=(X=0,Y=-2)
    InTransform_Scale=(X=1.3,Y=1.3)
    InTransform_Shear=(X=0,Y=0)
    InTransform_Angle=0
   )
   Reroute Node(InputPin:Wildcard) -> (OutputPin:Wildcard)
  ]}
 GetMemoryLimit {nodes [
   GetMemoryLimit() {is=title K2Node_FunctionEntry}
   n1 = Get MemoryLimit() -> (MemoryLimit:Scalable Float Structure)
   > Return Node!(
    execute
    ThermometerBarVaxValue:double="To Float (Integer)"(InInt=GetValueAsInteger!(Input:"Scalable Float Structure (by ref)"=@n1, Level:float=0) -> (ReturnValue:int=0))
   )
  ]}
 IsOverMemoryLimit {nodes [
   n1 = IsOverMemoryLimit!() -> (then, MemoryValue:int, Cost Of Preview Actor:int, Margin:int, MaxValue:int) {is=title K2Node_FunctionEntry}
   > Return Node!(
    execute
    bIsOverMemoryLimit:bool="integer >= integer"(A=int + int(A=@n1, B=@n1.Cost Of Preview Actor) {is=title K2Node_PromotableOperator}, B=int + int(A=@n1.Margin, B=@n1.MaxValue))
   )
  ]}
 IsOverMemoryMargin {nodes [
   n1 = IsOverMemoryMargin!() -> (then, MemoryValue:int, Cost Of Preview Actor:int, Margin:int, MaxValue:int) {is=title K2Node_FunctionEntry}
   > Set MemoryWithPreview(MemoryWithPreview:int=int + int(A=@n1, B=@n1.Cost Of Preview Actor)) -> (Output_Get:int=0)
   > Set MemoryLimitWithMargin(MemoryLimitWithMargin:int=int + int(A=@n1.Margin, B=@n1.MaxValue)) -> (Output_Get:int=0)
   n2 = Get MemoryWithPreview() -> (MemoryWithPreview:int=0)
   n3 = Get MemoryWithPreview() -> (MemoryWithPreview=0)
   n4 = Get MemoryLimitWithMargin() -> (MemoryLimitWithMargin:int=0)
   > Return Node!(execute, bIsOverMargin:bool=AND Boolean(A="integer > integer"(A=@n2, B=@n1.MaxValue), B="integer < integer"(A=@n3, B=@n4)))
  ]}
 UpdatePreviewBarVisuals {nodes [
   n1 = Inputs!() -> (execute, IsOverMemoryLimit:bool=~, IsOverMemoryMargin:bool=~)
   > n2 = Branch(Condition=@n1)
   > setTextColorred(self:Preview Bar Memory Object Reference=Get PreviewBar_Memory())
   > n3 = setWarningIconVisibility!(execute, self:Preview Bar Memory Object Reference=Get PreviewBar_Memory(), InVisibility:ESlateVisibility Enum=Visible) -> (then)
   Outputs!(then:Exec=[@n5.then, @n3.then, @n6.then])
   setTextColorWhite(execute=@n2.else, self:Preview Bar Memory Object Reference=Get PreviewBar_Memory())
   > n4 = Branch(Condition=@n1.IsOverMemoryMargin)
   > setWarningIconColorWhite(self:Preview Bar Memory Object Reference=Get PreviewBar_Memory())
   > n5 = setWarningIconVisibility(self=Get PreviewBar_Memory(), InVisibility=Visible)
   n6 = setWarningIconVisibility(execute=@n4.else, self=Get PreviewBar_Memory(), InVisibility=Collapsed)
   setWarningIconColorRed(self:Preview Bar Memory Object Reference=Get PreviewBar_Memory())
  ]}
]
variables [
 Cost Preview Of Selected Actor {type=integer}
 MemoryLimit {
  type=ScalableFloat
  default {
   registryType=None
   value=1
   curve {curveTable=/Game/Athena/Playlists/Creative/AthenaCompositeGD_Creative, rowName=Creative.Thermometer.UI.MemoryLimit}
  }
 }
]
widgets {
 info {
  inheritedWidgetCount=0
  parentClass=/Script/FortniteUI.FortCreativeVolumePerformanceMetricWidget
  rootWidgetClass=&ScaleBox
  widgetCount=3
 }
 widgets [
  ScaleBox_0 {widgetClassPath=&ScaleBox, widget=ScaleBox_0, properties {stretch=UserSpecified, userSpecifiedScale=0.8}}
  HorizontalBox_0 {widgetClassPath=&HorizontalBox, parent=ScaleBox_0, widget=HorizontalBox_0, slot {class=ScaleBoxSlot}}
  PreviewBar_Memory {
   widgetClassPath=/FortniteCoreUI/HUD/LTM/Playground/Thermometer/PreviewBar_Memory
   bIsVariable=true
   parent=HorizontalBox_0
   widget=PreviewBar_Memory
   properties {renderTransform {angle=0, scale=<1.2 1.2>, shear=<0 0>, translation=<0 0>}}
   slot {class=HorizontalBoxSlot, properties {horizontalAlignment=HAlign_Left, verticalAlignment=VAlign_Center}}
  }
 ]
}
