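# Creative-mode HUD widget that displays island memory usage as a titled thermometer bar.
# Memory-category metric updates become a ratio that sets bar progress and, scaled by a curve-table memory limit, its displayed value.
# Selected-actor cost updates feed the bar's preview as a fraction of the last metric maximum.
# At full capacity the title pulses red in a ping-pong loop, then plays back in reverse when usage drops.
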WidgetBlueprint /FortniteCoreUI/HUD/LTM/Playground/ThermometerV2/WBP_VolumePerformanceMetric_Thermometer
animations [Max {
  binding_count=1
  length_seconds=0.77
  object=&self:Max
  sequence {bindings [Title {
     id=#1
     possessed_class=CommonTextBlock
     tracks [{
       type=MovieSceneColorTrack
       display_name=Color and Opacity
       sections [{
         type=MovieSceneColorSection
         end_frame=15
         start_frame=0
         channels [
          Color.R {type=MovieSceneScriptingFloatChannel, keys [[0,1],[15.4,0.6939]]}
          Color.G {type=MovieSceneScriptingFloatChannel, keys [[0,1],[15.4,0]]}
          Color.B {type=MovieSceneScriptingFloatChannel, keys [[0,1],[15.4,0.0123]]}
          Color.A {type=MovieSceneScriptingFloatChannel, const [1,1]}
         ]
        }]
      }]
    }]}
 }]
graphs [
 EventGraph {nodes [
   n1 = Event On Cost Preview Of Selected Actor Updated!() -> (OutputDelegate, then, CostPreviewOfSelectedActor:int=0, HighestCellCostPreviewOfSelectedActor:int=0) {is=title K2Node_Event}
   n2 = Break Volume Performance Metrics!(VolumePerformanceMetrics:"Volume Performance Metrics Structure (by ref)"=Get LastVolumePerformanceMetrics() -> (LastVolumePerformanceMetrics:Volume Performance Metrics Structure)) -> (
    PerformanceValue:int
    PerformanceMaxValue:int
    PerformanceValueMargin:int
    PerformanceValueCellIndex:Int Point Structure
    PerformanceLowendThreshold:int
    PreviewDeltaValue:int
    Category:EFortBudgetCategory Enum
    PlayerID:int
    MostExpensivePerformanceValue:int
    MostExpensivePreviewDeltaValue:int
    GlobalMemoryValue:int
    MostExpensivePerformanceValueCellIndex:Int Point Structure
   ) {is=title K2Node_BreakStruct}
   > SetPreview Value!(
    execute
    self:Thermometer Bar Object Reference=Get ThermometerBar() -> (ThermometerBar:Thermometer Bar Object Reference)
    Preview value:double=float / float(A="To Float (Integer)"(InInt=@n1.CostPreviewOfSelectedActor), B="To Float (Integer)"(InInt=@n2.PerformanceMaxValue)) {is=title K2Node_CallFunction}
   ) -> (then)
   n3 = Event On Volume Performance Metrics Updated!() -> (OutputDelegate, then, VolumePerformanceMetrics:"Volume Performance Metrics Structure (by ref)") {is=title K2Node_Event}
   n4 = Break Volume Performance Metrics(VolumePerformanceMetrics=@n3.VolumePerformanceMetrics)
   > Branch(Condition="Equal (Enum)"(
     A:EFortBudgetCategory Enum=@n4.Category
     B:EFortBudgetCategory Enum=Get VolumePerformanceMetricCategory() -> (VolumePerformanceMetricCategory:EFortBudgetCategory Enum=Memory)
    ))
   > UpdateThermometerBar!(execute, self, Value:int=@n4, MaxValue:int=@n4.PerformanceMaxValue) -> (then)
   Event Construct()
   > SetMaxValue!(
    execute
    self:Thermometer Bar Object Reference=Get ThermometerBar()
    Value:double=GetMemoryLimitForThermometerBar!() -> (ThermometerBarVaxValue:double)
   ) -> (then)
  ]}
 GetMemoryLimitForThermometerBar {nodes [
   GetMemoryLimitForThermometerBar() {is=title K2Node_FunctionEntry}
   n1 = Get MemoryLimit() -> (MemoryLimit:Scalable Float Structure)
   > Return Node!(
    execute
    ThermometerBarVaxValue:double="To Float (Integer)"(InInt=GetValueAsInteger!(Input:"Scalable Float Structure (by ref)"=@n1, Level:float=0) -> (ReturnValue:int=0))
   )
  ]}
 UpdateThermometerBar {nodes [
   n1 = UpdateThermometerBar!() -> (then, Value:int, MaxValue:int) {is=title K2Node_FunctionEntry}
   n2 = "To Float (Integer)"(InInt=@n1)
   n3 = "To Float (Integer)"(InInt=@n1.MaxValue)
   > n4 = Set ValueRatio(ValueRatio:double=float / float(A="Max (Float)"(A=@n2, B=0), B="Max (Float)"(A=@n3, B=1))) -> (Output_Get:double)
   n5 = float * float(A=@n4, B=GetMemoryLimitForThermometerBar() {is=title K2Node_CallFunction})
   > SetCurrentValue!(
    execute
    self:Thermometer Bar Object Reference=Get ThermometerBar()
    Value:double="To Float (Integer)"(InInt=Round(A=@n5) {is=title K2Node_CallFunction})
   ) -> (then)
   > SetValueProgress!(execute, self:Thermometer Bar Object Reference=Get ThermometerBar(), Value:double=Get ValueRatio() -> (ValueRatio:double)) -> (then)
   > n6 = Branch(Condition="float >= float"(A=Get ValueRatio(), B=1) {is=title K2Node_CallFunction})
   > PlayAnimationTimeRange!(
    execute
    self:User Widget Object Reference
    InAnimation:Widget Animation Object Reference=Get Max() -> (Max:Widget Animation Object Reference)
    StartAtTime:float=0
    EndAtTime:float=0
    NumLoopsToPlay:int=0
    PlayMode:EUMGSequencePlayMode Enum=PingPong
    PlaybackSpeed:float=1
    bRestoreState:bool=false
   ) -> (then, ReturnValue:Widget Animation Handle Structure)
   Branch(execute=@n6.else, Condition=IsAnimationPlaying(InAnimation=Get Max()))
   > StopAnimation(InAnimation=Get Max())
   > PlayAnimationTimeRange(
    InAnimation=Get Max()
    StartAtTime=0
    EndAtTime=GetEndTime(self=Get Max())
    NumLoopsToPlay=1
    PlayMode=Reverse
    PlaybackSpeed=1
    bRestoreState=false
   )
  ]}
]
variables [MemoryLimit {
  type=ScalableFloat
  default {
   registryType=None
   value=1
   curve {curveTable=/Game/Athena/Playlists/Creative/AthenaCompositeGD_Creative, rowName=Creative.Thermometer.UI.MemoryLimit}
  }
 }]
widgets {
 info {
  inheritedWidgetCount=0
  parentClass=/Script/FortniteUI.FortCreativeVolumePerformanceMetricWidget
  rootWidgetClass=&ScaleBox
  widgetCount=5
 }
 widgets [
  ScaleBox_27 {widgetClassPath=&ScaleBox, widget=ScaleBox_27, properties {stretch=UserSpecified, userSpecifiedScale=0.8}}
  Overlay_26 {widgetClassPath=&Overlay, parent=ScaleBox_27, widget=Overlay_26, slot {class=ScaleBoxSlot}}
  VerticalBox_34 {widgetClassPath=&VerticalBox, parent=Overlay_26, widget=VerticalBox_34, slot {class=OverlaySlot}}
  Title {
   widgetClassPath=&CommonTextBlock
   parent=VerticalBox_34
   widget=Title
   properties {
    text=MEMORY USED
    font {
     bForceMonospaced=false
     bMaterialIsStencil=false
     fontMaterial=null
     fontObject=&Burbank
     letterSpacing=0
     monospacedWidth=1
     size=20
     skewAmount=0
     typefaceFontName=Burbank-PlayerPower
     outlineSettings {
      bApplyOutlineToDropShadows=true
      bMiteredCorners=false
      bSeparateFillAlpha=false
      outlineMaterial=null
      outlineSize=1
      outlineColor=<rgba 0 0 0 0.75>
     }
    }
    margin=<ltrb 0 -2 2 -1>
    renderTransform {angle=0, scale=<1 1>, shear=<0 0>, translation=<-30 0>}
   }
   slot {class=VerticalBoxSlot, properties {horizontalAlignment=HAlign_Center}}
  }
  ThermometerBar {
   widgetClassPath=/FortniteCoreUI/HUD/LTM/Playground/ThermometerBar
   bIsVariable=true
   parent=VerticalBox_34
   widget=ThermometerBar
   slot {class=VerticalBoxSlot}
  }
 ]
}
