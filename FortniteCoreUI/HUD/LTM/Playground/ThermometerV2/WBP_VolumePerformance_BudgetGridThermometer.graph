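# HUD widget for Creative volume performance metrics that stacks two budget grid thermometer bars for memory.
# PreConstruct titles the highest-cell bar "Global Impact" and enables the max-used pip on the regular bar.
# Metric updates in the Memory category feed the reported cell's values to the regular bar and the most expensive cell's values to the highest bar.
# SetupVisibility collapses or shows the widget and the highest bar through separate flags.
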WidgetBlueprint /FortniteCoreUI/HUD/LTM/Playground/ThermometerV2/WBP_VolumePerformance_BudgetGridThermometer
graphs [
 EventGraph {nodes [
   Event PreConstruct()
   > SetTitle!(
    execute
    self:WBP Volume Performance Metric Budget Grid Thermometer Bar Object Reference=Get BudgetGridThermometerBar_Highest() -> (BudgetGridThermometerBar_Highest:WBP Volume Performance Metric Budget Grid Thermometer Bar Object Reference)
    TitleText:Text="LOCTEXT(\"Global Impact\")"
   ) -> (then)
   > SetMaxUsedPipVisibility!(
    execute
    self:Budget Grid Thermometer Bar Object Reference=Get BudgetGridThermometerBar!(self:WBP Volume Performance Metric Budget Grid Thermometer Bar Object Reference=Get BudgetGridThermometerBar!() -> (BudgetGridThermometerBar:WBP Volume Performance Metric Budget Grid Thermometer Bar Object Reference)) -> (BudgetGridThermometerBar:Budget Grid Thermometer Bar Object Reference)
    bVisible:bool=true
   ) -> (then)
   n1 = Event On Volume Performance Metrics Updated!() -> (
    OutputDelegate
    then
    VolumePerformanceMetrics_PerformanceValue:int
    VolumePerformanceMetrics_PerformanceMaxValue:int
    VolumePerformanceMetrics_PerformanceValueMargin:int
    VolumePerformanceMetrics_PerformanceValueCellIndex:Int Point Structure
    VolumePerformanceMetrics_PerformanceLowendThreshold:int
    VolumePerformanceMetrics_PreviewDeltaValue:int
    VolumePerformanceMetrics_Category:EFortBudgetCategory Enum
    VolumePerformanceMetrics_PlayerID:int
    VolumePerformanceMetrics_MostExpensivePerformanceValue:int
    VolumePerformanceMetrics_MostExpensivePreviewDeltaValue:int
    VolumePerformanceMetrics_GlobalMemoryValue:int
    VolumePerformanceMetrics_MostExpensivePerformanceValueCellIndex:Int Point Structure
   ) {is=title K2Node_Event}
   > Branch(Condition="Equal (Enum)"(A:EFortBudgetCategory Enum=@n1.VolumePerformanceMetrics_Category, B:EFortBudgetCategory Enum=Memory))
   > n2 = Sequence()
   > UpdatePerformanceMetrics!(
    execute
    self:WBP Volume Performance Metric Budget Grid Thermometer Bar Object Reference=Get BudgetGridThermometerBar()
    Value:int=@n1.VolumePerformanceMetrics_PerformanceValue
    MaxValue:int=@n1.VolumePerformanceMetrics_PerformanceMaxValue
    Preview Delta Value:int=@n1.VolumePerformanceMetrics_PreviewDeltaValue
    Coordinates:Int Point Structure=@n1.VolumePerformanceMetrics_PerformanceValueCellIndex
    Margin:int=@n1.VolumePerformanceMetrics_PerformanceValueMargin
    HighestCellValue:int=@n1.VolumePerformanceMetrics_MostExpensivePerformanceValue
    HighestCellPreviewDeltaValue:int=@n1.VolumePerformanceMetrics_MostExpensivePreviewDeltaValue
    GlobalMemoryValue:int=@n1.VolumePerformanceMetrics_GlobalMemoryValue
   ) -> (then)
   UpdatePerformanceMetrics(
    execute=@n2.then_1
    self=Get BudgetGridThermometerBar_Highest()
    Value=@n1.VolumePerformanceMetrics_MostExpensivePerformanceValue
    MaxValue=@n1.VolumePerformanceMetrics_PerformanceMaxValue
    Preview Delta Value=@n1.VolumePerformanceMetrics_MostExpensivePreviewDeltaValue
    Coordinates=@n1.VolumePerformanceMetrics_MostExpensivePerformanceValueCellIndex
    Margin=@n1.VolumePerformanceMetrics_PerformanceValueMargin
    HighestCellValue=@n1.VolumePerformanceMetrics_MostExpensivePerformanceValue
    HighestCellPreviewDeltaValue=@n1.VolumePerformanceMetrics_MostExpensivePreviewDeltaValue
    GlobalMemoryValue=@n1.VolumePerformanceMetrics_GlobalMemoryValue
   )
   Event Construct()
   Event Tick()
  ]}
 SetupVisibility {nodes [
   n1 = SetupVisibility!() -> (then, UseBudgetGridThermometer:bool=~, DisplayMaxMemoryValueThermometer:bool=~) {is=title K2Node_FunctionEntry}
   > SetVisibility(InVisibility=Select(Option 0=Collapsed, Option 1=HitTestInvisible, Index=@n1))
   > SetVisibility(
    self=Get BudgetGridThermometerBar_Highest()
    InVisibility=Select(Option 0=Collapsed, Option 1=HitTestInvisible, Index=@n1.DisplayMaxMemoryValueThermometer)
   )
  ]}
]
variables []
widgets {
 info {
  inheritedWidgetCount=0
  parentClass=/Script/FortniteUI.FortCreativeVolumePerformanceMetricWidget
  rootWidgetClass=&ScaleBox
  widgetCount=4
 }
 widgets [
  ScaleBox_0 {widgetClassPath=&ScaleBox, widget=ScaleBox_0, properties {stretch=UserSpecified}}
  VerticalBox_18 {widgetClassPath=&VerticalBox, parent=ScaleBox_0, widget=VerticalBox_18, slot {class=ScaleBoxSlot}}
  BudgetGridThermometerBar_Highest {
   widgetClassPath=&/WBP_VolumePerformanceMetric_BudgetGridThermometerBar
   bIsVariable=true
   parent=VerticalBox_18
   widget=BudgetGridThermometerBar_Highest
   properties {bShowHighestMemoryCell=true, visibility=HitTestInvisible}
   slot {class=VerticalBoxSlot, properties {padding=<ltrb 0 0 0 16>}}
  }
  BudgetGridThermometerBar {
   widgetClassPath=&WBP_VolumePerformanceMetric_BudgetGridThermometerBar
   bIsVariable=true
   parent=VerticalBox_18
   widget=BudgetGridThermometerBar
   slot {class=VerticalBoxSlot}
  }
 ]
}
