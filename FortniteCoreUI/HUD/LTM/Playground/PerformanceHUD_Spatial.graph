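# Creative-mode HUD widget that shows volume memory budget thermometers, a spatial heatmap, cost preview and save status.
# Entering a volume selects island, heatmap or budget-grid thermometers, updates the map manager and requests server heatmap bounds.
# Metric events track global and group usage, feeding aim location and group value into the MPC_heatmap parameter collection.
# Selection cost previews trigger over-budget warnings and toggle the heatmap's throw-preview state.
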WidgetBlueprint /FortniteCoreUI/HUD/LTM/Playground/PerformanceHUD_Spatial
animations [BarFull_Server {
  binding_count=1
  length_seconds=0.773
  object=&self:BarFull_Server
  sequence {bindings [ThermometerTitle_Server {
     id=#1
     display_name=WBP_VolumePerformanceMetric_HeatmapThermometer
     possessed_class=CommonTextBlock
     tracks [{
       type=MovieSceneColorTrack
       display_name=Color and Opacity
       sections [{
         type=MovieSceneColorSection
         end_frame=15
         start_frame=0
         channels [
          Color.R {type=MovieSceneScriptingFloatChannel, keys [[0,1],[15.324,0.6939]]}
          Color.G {type=MovieSceneScriptingFloatChannel, keys [[0,1],[15.324,0]]}
          Color.B {type=MovieSceneScriptingFloatChannel, keys [[0,1],[15.324,0.0123]]}
          Color.A {type=MovieSceneScriptingFloatChannel, const [1,2]}
         ]
        }]
      }]
    }]}
 }]
graphs [
 ChangeThermometerWarningVisibility {nodes [
   n1 = ChangeThermometerWarningVisibility!() -> (then, Visible:bool=~) {is=title K2Node_FunctionEntry}
   n2 = Get ThermometerArea_WarningTips() -> (ThermometerArea_WarningTips:Scale Box Object Reference)
   n3 = Select(Option 0=Collapsed, Option 1=HitTestInvisible, Index=@n1)
   > Branch(Condition="Not Equal (Enum)"!(
     A:ESlateVisibility Enum=GetVisibility!(self:Widget Object Reference=@n2) -> (ReturnValue:ESlateVisibility Enum=Visible)
     B:ESlateVisibility Enum=@n3
    ) -> (ReturnValue:bool))
   > SetVisibility(self=Get ThermometerArea_WarningTips(), InVisibility=@n3)
   > n4 = Branch(Condition=@n1)
   n5 = Get ThermometerWarningTips() -> (ThermometerWarningTips:Thermometer Warning Tips Object Reference)
   > Set Grouptext(self:Thermometer Warning Tips Object Reference=@n5) {is=title K2Node_CallFunction}
   > PlayWarningLoad(self:Thermometer Warning Tips Object Reference=@n5)
   PlayWarningOut(execute=@n4.else, self:Thermometer Warning Tips Object Reference=@n5)
  ]}
 EventGraph {nodes [
   Event Construct()
   > SetVisibility(
    self=[Get ThermometerArea_Heatmap() -> (ThermometerArea_Heatmap:Scale Box Object Reference), Get ThermometerArea_WarningTips()]
    InVisibility=Collapsed
   )
   > n1 = Cast To FortPlayerControllerGameplay!(execute, Object=GetOwningPlayer()) -> (then, CastFailed, AsFort Player Controller Gameplay:Fort Player Controller Gameplay Object Reference) {is=title K2Node_DynamicCast}
   > n2 = Bind Event to On Creative UIMetrics Display Changed(self:Fort Player Controller Gameplay Object Reference=@n1, Delegate=Create Event()) {is=title K2Node_AddDelegate}
   Sequence(execute=[@n2.then, @n3.then])
   > SetWidgetPositionOnPlatforms!(
    execute
    self
    isApplyToBattleLab:bool=false
    MobilePos:Vector 2D Structure=(X=-30,Y=175)
    DefaultPos:Vector 2D Structure=(X=-320,Y=50)
    WidgetToRepos:Widget Object Reference=Get CreativeSaveStatusDisplay() -> (CreativeSaveStatusDisplay:Creative Save Status Display Object Reference)
   ) -> (then)
   n3 = PrintString!(
    execute=@n1.CastFailed
    InString="PerformanceHUD_Spatial (Event Constuct) failed  to cast to FortPlayerControllerGameplay"
    bPrintToScreen
    bPrintToLog
    TextColor=(R=1,G=0,B=0,A=1)
    Duration
    Key
   ) -> (then)
   Event HideUIMetrics() {is=title K2Node_Event}
   > Set CurrentMemoryValue(CurrentMemoryValue:int=0) -> (Output_Get:int=0)
   > Set CurrentMemoryValueAfterRound(CurrentMemoryValueAfterRound:int=0) -> (Output_Get:int=0)
   > SetScalarParameterValue!(
    execute
    Collection:Material Parameter Collection Object Reference=/Game/Athena/HUD/LTM/Playground/Thermometer/MPC_heatmap
    ParameterName:Name=HeatMapSwitch
    ParameterValue:float=0
   ) -> (then) {is=title K2Node_CallMaterialParameterCollectionFunction}
   > ClearRenderTarget2D!(
    execute
    TextureRenderTarget:Texture Render Target 2D Object Reference=/Game/Athena/HUD/LTM/Playground/Thermometer/RT_Heatmapcode
    ClearColor:Linear Color Structure=(R=0,G=0,B=0,A=1)
   ) -> (then)
   > SetVisibility(InVisibility=Hidden)
   > n4 = SetVisibility(self=[Get ThermometerArea_Heatmap(), Get ThermometerArea_WarningTips()], InVisibility=Collapsed)
   Do Once(execute=@n10.then_1, Reset=@n4.then)
   > SetScalarParameterValue(Collection=&MPC_heatmap, ParameterName=GroupRadius, ParameterValue="To Float (Integer)"(InInt=0))
   > Branch(Condition=GetValueAsBool!(Input:"Scalable Float Structure (by ref)"=Get ShouldShowHeatmap() -> (ShouldShowHeatmap:Scalable Float Structure), Level:float=0) -> (ReturnValue:bool))
   n5 = Get Thermometer_HeatmapController() -> (Thermometer_HeatmapController:Thermometer Heatmap Controller Object Reference)
   > SetHeatmapUIScale(self:Thermometer Heatmap Controller Object Reference=@n5)
   > SetVisibility(self=@n5, InVisibility=HitTestInvisible)
   > SetHeatmapVisibility!(execute, self:Thermometer Heatmap Controller Object Reference=@n5, EnableHeatmapVisibility:bool=true) -> (then)
   > n6 = Branch(Condition=IsMobileGame())
   n7 = Get ThermometerArea_Heatmap()
   > SetVisibility(self=@n7, InVisibility=Hidden)
   SetVisibility(execute=@n6.else, self=@n7, InVisibility=HitTestInvisible)
   Event SetUIMetrics!() -> (OutputDelegate, then, VolumePerformanceMetrics:"Volume Performance Metrics Structure (by ref)") {is=title K2Node_Event}
   > n8 = Branch(Condition=IsHeatmapEnabled!(self:Creative Performance HUDBase Object Reference) -> (ReturnValue:bool))
   > SetVisibility(InVisibility=HitTestInvisible)
   SetVisibility(execute=@n8.else, InVisibility=Collapsed)
   n9 = Event SetSimpleUIMetrics!() -> (OutputDelegate, then, SimplePerformanceMetrics:"Simple Metric Information Structure (by ref)") {is=title K2Node_Event}
   > Branch(Condition=IsHeatmapEnabled())
   > n10 = Sequence()
   n11 = Break Simple Metric Information!(SimpleMetricInformation:"Simple Metric Information Structure (by ref)"=@n9.SimplePerformanceMetrics) -> (
    CategoryNames:Array of EFortBudgetCategory Enums
    CategoryValues:Array of Integers
    CategoryBudgets:Array of Integers
    Location:Vector=~
    NormalizedGroupBudgetValue:float=~
    GroupTotalBudget:int
    NeighbourGroupCenter:Vector=~
    NeighbourGroupBudgetValue:Array of Bytes
   ) {is=title K2Node_BreakStruct}
   > Branch(Condition="Not Equal (Integer)"!(A:int=Length(TargetArray:Array of EFortBudgetCategory Enums=@n11), B:int=0) -> (ReturnValue:bool) {is=title K2Node_CallFunction})
   > n12 = Set GroupMemoryProgress(GroupMemoryProgress:double=@n11.NormalizedGroupBudgetValue) -> (Output_Get:double)
   > n13 = Branch(Condition=IsMobileGame())
   Branch(execute=[@n15.then, @n13.then], Condition=Get IsInPreviewMode() -> (IsInPreviewMode:bool))
   n14 = BreakVector(InVec=@n11.Location)
   > SetScalarParameterValue(Collection=&MPC_heatmap, ParameterName=AimTargetX, ParameterValue=@n14)
   > SetScalarParameterValue(Collection=&MPC_heatmap, ParameterName=AimTargetY, ParameterValue=@n14.Y)
   n15 = SetScalarParameterValue(execute=@n13.else, Collection=&MPC_heatmap, ParameterName=GroupNewValue, ParameterValue=@n12)
   n16 = Event On Client Enter Volume!() -> (OutputDelegate, then, ClientPlayerState:Player State Object Reference, Volume:Fort Volume Object Reference) {is=title K2Node_Event}
   > Is Valid(InputObject=@n16.Volume)
   n17 = GetObjectTrackingComponent!(self:Fort Volume Object Reference=@n16.Volume) -> (ReturnValue:Fort Volume Object Tracking Component Object Reference)
   > Is Valid(InputObject=@n17)
   > n18 = Sequence()
   n19 = IsHeatmapEnabled(self:Fort Volume Object Tracking Component Object Reference=@n17)
   > n20 = Branch(Condition=@n19)
   > SetVisibility(self=Get HeatmapThermometerPanel() -> (HeatmapThermometerPanel:Canvas Panel Object Reference), InVisibility=Visible)
   > SetVisibility(self=Get IslandThermometerPanel() -> (IslandThermometerPanel:Canvas Panel Object Reference), InVisibility=Collapsed)
   n21 = UseBudgetGridThermometer!(self:Fort Volume Object Tracking Component Object Reference=@n17) -> (ReturnValue:bool)
   > SetupVisibility!(
    execute
    self:WBP Volume Performance Budget Grid Thermometer Object Reference=Get WBP_VolumePerformance_BudgetGridThermometer() -> (WBP_VolumePerformance_BudgetGridThermometer:WBP Volume Performance Budget Grid Thermometer Object Reference)
    UseBudgetGridThermometer:bool=@n21
    DisplayMaxMemoryValueThermometer:bool=DisplayMaxMemoryValueThermometer!(self:Fort Volume Object Tracking Component Object Reference=@n17) -> (ReturnValue:bool)
   ) -> (then)
   > n22 = SetVisibility(
    self=Get WBP_VolumePerformanceMetric_HeatmapThermometer() -> (WBP_VolumePerformanceMetric_HeatmapThermometer:WBP Volume Performance Metric Heatmap Thermometer Object Reference)
    InVisibility=Select(Option 0=HitTestInvisible, Option 1=Collapsed, Index=@n21)
   )
   Branch(execute=[@n22.then, @n24.then], Condition=@n19)
   > n23 = Cast To FortPlayerControllerAthena(Object=GetOwningPlayer())
   > ServerUpdateHeatmapBoundsForCurrentVolume(self:Fort Player Controller Athena Object Reference=@n23)
   SetVisibility(execute=@n20.else, self=Get IslandThermometerPanel(), InVisibility=HitTestInvisible)
   > n24 = SetVisibility(self=Get HeatmapThermometerPanel(), InVisibility=Collapsed)
   n25 = Cast To FortInGameMapManagerCreative!(execute=@n18.then_1, Object=GetUIMapManager!() -> (ReturnValue:Fort In Game Map Manager Object Reference)) -> (then, CastFailed, AsFort in Game Map Manager Creative:Fort In Game Map Manager Creative Object Reference) {is=title K2Node_DynamicCast}
   > SetCurrentVolume!(execute, self:Fort In Game Map Manager Creative Object Reference=@n25, InVolume:Fort Volume Object Reference=@n16.Volume) -> (then)
   n26 = Branch(Condition=IsHeatmapEnabled(self:Creative Performance HUDBase Object Reference))
   n27 = Event RefreshCostPreviewOfSelectedActor!() -> (OutputDelegate, then, CostPreviewOfSelection:int=0, isThrowingPlayset:bool) {is=title K2Node_Event}
   n28 = Branch(execute=@n26.then, Condition="Not Equal (Integer)"(A=@n27.CostPreviewOfSelection, B=0))
   > n29 = Sequence()
   n30 = Get GroupMemoryProgress() -> (GroupMemoryProgress:double)
   n31 = float / float(
    A="To Float (Integer)"(InInt=@n27.CostPreviewOfSelection)
    B="To Float (Integer)"(InInt=Get MaxMemoryBudgetValue() -> (MaxMemoryBudgetValue:int=0))
   ) {is=title K2Node_CallFunction}
   n32 = float + float(A=Get GlobalMemoryProgress() -> (GlobalMemoryProgress:double), B=@n31) {is=title K2Node_CommutativeAssociativeBinaryOperator}
   > ChangeThermometerWarningVisibility!(execute, self, Visible:bool=OR Boolean(A="float >= float"(A=@n32, B=1) {is=title K2Node_CallFunction}, B="float >= float"(A=@n30, B=1))) -> (then) {is=title K2Node_CallFunction}
   n33 = Set IsInPreviewMode(execute=@n29.then_1, IsInPreviewMode:bool=@n27.isThrowingPlayset) -> (Output_Get:bool)
   > n34 = Set IsInPreviewMode!(execute, IsInPreviewMode=@n33, self:Thermometer Heatmap Controller Object Reference=Get Thermometer_HeatmapController()) -> (then, Output_Get)
   > n35 = Branch(Condition=@n34)
   > SetScalarParameterValue(Collection=&MPC_heatmap, ParameterName=ThrowItems, ParameterValue=1)
   > Do Once(Reset=@n36.then)
   > DisplayHeatmap(self:Thermometer Heatmap Controller Object Reference=Get Thermometer_HeatmapController())
   > n36 = Delay(Duration=60)
   SetScalarParameterValue(execute=@n35.else, Collection=&MPC_heatmap, ParameterName=ThrowItems, ParameterValue=0)
   Set IsInPreviewMode(execute=@n28.else, IsInPreviewMode=false, self=Get Thermometer_HeatmapController())
   > Set IsInPreviewMode!(execute, IsInPreviewMode=false) -> (then, Output_Get)
   > SetScalarParameterValue(Collection=&MPC_heatmap, ParameterName=ThrowItems, ParameterValue=0)
   > SetVisibility(self=Get ThermometerArea_WarningTips(), InVisibility=Collapsed)
   > PlayWarningOut(self=Get ThermometerWarningTips())
   n37 = Break Volume Performance Metrics!(VolumePerformanceMetrics:"Volume Performance Metrics Structure (by ref)"=Reroute Node(InputPin:"Volume Performance Metrics Structure (by ref)") -> (OutputPin:"Volume Performance Metrics Structure (by ref)")) -> (
    PerformanceValue:int
    PerformanceMaxValue:int
    PerformanceValueMargin:int
    PerformanceValueCellIndex:Int Point Structure
    PerformanceLowendThreshold:int
    PreviewDeltaValue:int
    Category:EFortBudgetCategory Enum
    PlayerID:int
    MostExpensivePerformanceValue:int
    MostExpensivePreviewDeltaValue:int
    GlobalMemoryValue:int
    MostExpensivePerformanceValueCellIndex:Int Point Structure
   ) {is=title K2Node_BreakStruct}
   Set MaxMemoryBudgetValue(MaxMemoryBudgetValue:int=@n37.PerformanceMaxValue) -> (Output_Get:int=0)
   > Set CurrentMemoryValue(CurrentMemoryValue=@n37) -> (Output_Get=0)
   n38 = Get MaxMemoryBudgetValue() -> (MaxMemoryBudgetValue=0)
   > UpdateUIMetrics_Original!(
    execute
    self
    PerformanceValue:int=@n37
    PerformanceMaxValue:int=@n37.PerformanceMaxValue
    PerformanceMaxBarValue:double="To Float (Integer)"(InInt=int / int!(A:int=@n38, B:int=2000) -> (ReturnValue:int=0) {is=title K2Node_CallFunction})
    PerformanceCategory:EFortBudgetCategory Enum=@n37.Category
    isServerMemory:bool=true
   ) -> (then)
   Reroute Node(InputPin:double=@n31) -> (OutputPin:double)
  ]}
 On Cost Preview Visibility Changed {nodes [
   n1 = On Cost Preview Visibility Changed!() -> (then, UIMetricsDisplayIndex:int) {is=title K2Node_FunctionEntry}
   > n2 = Switch on Int!(execute:Exec, Selection:int=@n1) -> (Default:Exec, 0:Exec, 1:Exec, 2:Exec)
   n3 = Set CostPreviewerIndex(execute=[@n2.0, @n2.Default], CostPreviewerIndex:int=1) -> (Output_Get:int=0)
   SetCostpreviewVisibility!(execute=[@n5.then, @n3.then, @n4.then], self, visibility:int=Get CostPreviewerIndex() -> (CostPreviewerIndex:int=0)) -> (then)
   n4 = Set CostPreviewerIndex(execute=@n2.1, CostPreviewerIndex=2) -> (Output_Get=0)
   n5 = Set CostPreviewerIndex(execute=@n2.2, CostPreviewerIndex=0) -> (Output_Get=0)
  ]}
 SetCostpreviewVisibility {nodes [
   n1 = SetCostpreviewVisibility!() -> (then, visibility:int) {is=title K2Node_FunctionEntry}
   > n2 = Switch on Int(Selection=@n1)
   SetVisibility(
    execute=[@n2.0, @n2.Default]
    self=Get WBP_VolumePerformanceMetric_PreviewCost() -> (WBP_VolumePerformanceMetric_PreviewCost:WBP Volume Performance Metric Preview Cost Object Reference)
    InVisibility=Collapsed
   )
   n3 = Get WBP_VolumePerformanceMetric_PreviewCost()
   SetVisibility(execute=@n2.1, self=@n3, InVisibility=SelfHitTestInvisible)
   n4 = SlotAsCanvasSlot!(Widget:Widget Object Reference=@n3) -> (ReturnValue:Canvas Panel Slot Object Reference)
   > SetPosition!(execute, self:Canvas Panel Slot Object Reference=@n4, InPosition:Vector 2D Structure=(X=40,Y=0)) -> (then)
   SetVisibility(execute=@n2.2, self=@n3, InVisibility=SelfHitTestInvisible)
   > SetPosition(self=@n4, InPosition=(X=-100,Y=-300))
  ]}
 SetWidgetPositionOnPlatforms {nodes [
   n1 = SetWidgetPositionOnPlatforms!() -> (then, isApplyToBattleLab:bool=~, MobilePos:Vector 2D Structure, DefaultPos:Vector 2D Structure, WidgetToRepos:Widget Object Reference) {is=title K2Node_FunctionEntry}
   > n2 = Branch(Condition=@n1)
   > n3 = Cast To FortGameStateAthena!(execute, Object=GetGameState!() -> (ReturnValue:Game State Base Object Reference)) -> (then, CastFailed, AsFort Game State Athena:Fort Game State Athena Object Reference) {is=title K2Node_DynamicCast}
   > Is Valid(InputObject=@n3)
   > n4 = Branch(Condition=IsMobileGame())
   n5 = Set TargetPos(execute=[@n4.then, @n8.then], TargetPos:Vector 2D Structure=@n1.MobilePos) -> (Output_Get:Vector 2D Structure)
   n6 = SlotAsCanvasSlot(Widget=@n1.WidgetToRepos)
   Is Valid(exec=[@n7.then, @n5.then], InputObject=@n6)
   > SetPosition(self=@n6, InPosition=Get TargetPos() -> (TargetPos:Vector 2D Structure))
   > Return Node()
   n7 = Set TargetPos(execute=[@n8.else, @n4.else], TargetPos=@n1.DefaultPos)
   n8 = Branch(execute=@n2.else, Condition=IsMobileGame())
  ]}
 UpdateUIMetrics_Original {nodes [
   n1 = UpdateUIMetrics_Original!() -> (
    then
    PerformanceValue:int
    PerformanceMaxValue:int
    PerformanceMaxBarValue:double=~
    PerformanceCategory:EFortBudgetCategory Enum=Memory
    isServerMemory:bool=~
   ) {is=title K2Node_FunctionEntry}
   > Branch(Condition="Equal (Enum)"(A:EFortBudgetCategory Enum=@n1.PerformanceCategory, B:EFortBudgetCategory Enum=Memory))
   > n2 = Branch(Condition=NOT Boolean(A=@n1.isServerMemory))
   n3 = "To Float (Integer)"(InInt=@n1)
   n4 = float / float(A="Clamp (Float)"(Value=@n3, Min=0, Max=@n3), B="To Float (Integer)"(InInt=@n1.PerformanceMaxValue))
   n5 = Round(A=float * float(A=@n4, B=@n1.PerformanceMaxBarValue)) {is=title K2Node_CallFunction}
   Set CurrentMemoryValueAfterRound(execute=@n2.else, CurrentMemoryValueAfterRound=@n5) -> (Output_Get=0)
   > Set GlobalMemoryProgress(GlobalMemoryProgress:double=@n4) -> (Output_Get:double)
   > Set BarFullAnimation(BarFullAnimation:Widget Animation Object Reference=Get BarFull_Server() -> (BarFull_Server:Widget Animation Object Reference)) -> (Output_Get:Widget Animation Object Reference)
   > n6 = Branch(Condition="float >= float"(A=@n4, B=1))
   > PlayAnimationTimeRange!(
    execute
    self:User Widget Object Reference
    InAnimation:Widget Animation Object Reference=Get BarFullAnimation() -> (BarFullAnimation:Widget Animation Object Reference)
    StartAtTime:float=0
    EndAtTime:float=0
    NumLoopsToPlay:int=0
    PlayMode:EUMGSequencePlayMode Enum=PingPong
    PlaybackSpeed:float=1
    bRestoreState:bool=false
   ) -> (then, ReturnValue:Widget Animation Handle Structure)
   Branch(execute=@n6.else, Condition=IsAnimationPlaying(InAnimation=Get BarFullAnimation()))
   > StopAnimation(InAnimation=Get BarFullAnimation())
   n7 = Get BarFullAnimation()
   > PlayAnimationTimeRange(InAnimation=@n7, StartAtTime=0, EndAtTime=GetEndTime(self=@n7), NumLoopsToPlay=1, PlayMode=Reverse, PlaybackSpeed=1, bRestoreState=false)
   "To Float (Integer)"(InInt=@n5)
  ]}
]
variables [
 CurrentMemoryValue {type=integer}
 CurrentMemoryValueAfterRound {type=integer}
 GlobalMemoryProgress {type=number}
 GroupMemoryProgress {type=number}
 IsInPreviewMode {type=boolean}
 MaxMemoryBudgetValue {type=integer}
 MaxServerBarValue {type=number, default=100000}
 ShouldShowHeatmap {
  type=ScalableFloat
  default {
   registryType=None
   value=1
   curve {curveTable=/Game/Athena/Playlists/Creative/AthenaCompositeGD_Creative, rowName=Creative.Thermometer.3DHeatmap.Enabled}
  }
 }
]
widgets {
 info {inheritedWidgetCount=1, parentClass=/Script/FortniteUI.CreativePerformanceHUDBase, rootWidgetClass=&CanvasPanel, widgetCount=15}
 widgets [
  CanvasPanel_0 {widgetClassPath=&CanvasPanel, widget=CanvasPanel_0}
  IslandThermometerPanel {
   widgetClassPath=&CanvasPanel
   bIsVariable=true
   parent=CanvasPanel_0
   widget=IslandThermometerPanel
   slot {class=CanvasPanelSlot, properties {layoutData {alignment=<0 0>, anchors {maximum=<1 1>, minimum=<0 0>}, offsets=<ltrb 0 0 0 0>}}}
  }
  WBP_VolumePerformanceMetric_Thermometer {
   widgetClassPath=/FortniteCoreUI/HUD/LTM/Playground/ThermometerV2/WBP_VolumePerformanceMetric_Thermometer
   parent=IslandThermometerPanel
   widget=WBP_VolumePerformanceMetric_Thermometer
   slot {
    class=CanvasPanelSlot
    properties {bAutoSize=true, layoutData {alignment=<0.5 0>, anchors {maximum=<0.5 0>, minimum=<0.5 0>}, offsets=<ltrb 0 70 100 30>}}
   }
  }
  ThermometerArea_WarningTips {
   widgetClassPath=&ScaleBox
   bIsVariable=true
   parent=IslandThermometerPanel
   widget=ThermometerArea_WarningTips
   properties {stretch=UserSpecified, userSpecifiedScale=0.8, visibility=Collapsed}
   slot {
    class=CanvasPanelSlot
    properties {bAutoSize=true, layoutData {alignment=<1 0>, anchors {maximum=<0.5 0.5>, minimum=<0.5 0.5>}, offsets=<ltrb 270 -120 100 30>}}
   }
  }
  ThermometerWarningTips {
   widgetClassPath=/FortniteCoreUI/Athena/HUD/LTM/Playground/Thermometer/ThermometerWarningTips
   bIsVariable=true
   parent=ThermometerArea_WarningTips
   widget=ThermometerWarningTips
   slot {class=ScaleBoxSlot}
  }
  HeatmapThermometerPanel {
   widgetClassPath=&CanvasPanel
   bIsVariable=true
   parent=CanvasPanel_0
   widget=HeatmapThermometerPanel
   slot {class=CanvasPanelSlot, properties {layoutData {alignment=<0 0>, anchors {maximum=<1 1>, minimum=<0 0>}, offsets=<ltrb 0 0 0 0>}}}
  }
  WBP_VolumePerformanceMetric_HeatmapThermometer {
   widgetClassPath=/FortniteCoreUI/HUD/LTM/Playground/ThermometerV2/WBP_VolumePerformanceMetric_HeatmapThermometer
   bIsVariable=true
   parent=HeatmapThermometerPanel
   widget=WBP_VolumePerformanceMetric_HeatmapThermometer
   slot {
    class=CanvasPanelSlot
    properties {bAutoSize=true, layoutData {alignment=<0.5 0>, anchors {maximum=<0.5 0>, minimum=<0.5 0>}, offsets=<ltrb 0 70 100 30>}}
   }
  }
  ThermometerArea_Heatmap {
   widgetClassPath=&ScaleBox
   bIsVariable=true
   parent=HeatmapThermometerPanel
   widget=ThermometerArea_Heatmap
   properties {stretch=ScaleToFill, visibility=Collapsed}
   slot {
    class=CanvasPanelSlot
    properties {bAutoSize=true, layoutData {alignment=<1.05 -1>, anchors {maximum=<1 0.5>, minimum=<1 0.5>}, offsets=<ltrb 0 0 100 32>}}
   }
  }
  Thermometer_HeatmapController {
   widgetClassPath=/FortniteCoreUI/Athena/HUD/LTM/Playground/Thermometer/Thermometer_HeatmapController
   bIsVariable=true
   parent=ThermometerArea_Heatmap
   widget=Thermometer_HeatmapController
   properties {visibility=HitTestInvisible}
   slot {class=ScaleBoxSlot, properties {horizontalAlignment=HAlign_Right, verticalAlignment=VAlign_Fill}}
  }
  ScaleBox_0 {
   widgetClassPath=&ScaleBox
   parent=HeatmapThermometerPanel
   widget=ScaleBox_0
   properties {stretch=UserSpecified, userSpecifiedScale=0.8}
   slot {
    class=CanvasPanelSlot
    properties {bAutoSize=true, layoutData {alignment=<1 0>, anchors {maximum=<0.5 0.5>, minimum=<0.5 0.5>}, offsets=<ltrb 270 -120 100 30>}}
   }
  }
  WBP_HeatmapOverBudget_Warning {
   widgetClassPath=/FortniteCoreUI/Athena/HUD/LTM/Playground/ThermometerV2/WBP_HeatmapOverBudget_Warning
   parent=ScaleBox_0
   widget=WBP_HeatmapOverBudget_Warning
   properties {visibility=Collapsed}
   slot {class=ScaleBoxSlot}
  }
  WBP_VolumePerformance_BudgetGridThermometer {
   widgetClassPath=/FortniteCoreUI/HUD/LTM/Playground/ThermometerV2/WBP_VolumePerformance_BudgetGridThermometer
   bIsVariable=true
   parent=HeatmapThermometerPanel
   widget=WBP_VolumePerformance_BudgetGridThermometer
   slot {
    class=CanvasPanelSlot
    properties {bAutoSize=true, layoutData {alignment=<0.5 0>, anchors {maximum=<0.5 0>, minimum=<0.5 0>}, offsets=<ltrb 0 70 100 30>}}
   }
  }
  CreativeSaveStatusDisplay {
   widgetClassPath=/FortniteCoreUI/Athena/HUD/LTM/Playground/CreativeSaveStatusDisplay
   bIsVariable=true
   parent=CanvasPanel_0
   widget=CreativeSaveStatusDisplay
   slot {
    class=CanvasPanelSlot
    properties {bAutoSize=true, layoutData {alignment=<1 0>, anchors {maximum=<1 0>, minimum=<1 0>}, offsets=<ltrb -378.7838 40 100 30>}}
   }
  }
  WBP_VolumePerformanceMetric_PreviewCost {
   widgetClassPath=/FortniteCoreUI/HUD/LTM/Playground/ThermometerV2/WBP_VolumePerformanceMetric_PreviewCost
   bIsVariable=true
   parent=CanvasPanel_0
   widget=WBP_VolumePerformanceMetric_PreviewCost
   slot {
    class=CanvasPanelSlot
    properties {bAutoSize=true, layoutData {alignment=<0 0>, anchors {maximum=<0.5 0.5>, minimum=<0.5 0.5>}, offsets=<ltrb 0 0 100 19.2>}}
   }
  }
  TutorialHighlight {widgetClassPath=&TutorialHighlight}
 ]
}
